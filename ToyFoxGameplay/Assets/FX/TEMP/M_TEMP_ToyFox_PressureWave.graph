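# Temporary translucent, unlit ToyFox pressure wave material for Niagara mesh particles, using index-of-refraction distortion.
# A fresnel-style edge term, maxed with inverted depth fade on higher-end platforms only, indexes a curve atlas row.
# Its channels drive refraction between low and high values, particle-alpha-scaled opacity and particle-color-scaled emissive strength.
# Particle random varies refraction strength from neutral, and a tiled normal texture has adjustable flattening.
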Material /ToyFoxGameplay/Assets/FX/TEMP/M_TEMP_ToyFox_PressureWave
graph {
 properties {
  bAutomaticallySetUsageInEditor=false
  bUsedWithNiagaraMeshParticles=true
  blendMode=BLEND_Translucent
  refractionMethod=RM_IndexOfRefraction
  shadingModel=MSM_Unlit
  usedShadingModels=Unlit
 }
 nodes [
  DepthFade()
  n1 = CameraVectorWS()
  n2 = VertexNormalWS()
  n3 = OneMinus(None=Abs(None=DotProduct(A=@n1, B=@n2)))
  n4 = ScalarParameter() {parameterName=Depth fade}
  n5 = Max(A=@n3, B=OneMinus(None=DepthFade(FadeDistance=@n4)))
  n6 = CurveAtlasRowParameter(CurveTime=MaterialFunctionCall("Xb1,PS4"=@n5, "PC-High&&Epic,XSX,PS5"=@n5, "PC-Perf,Android"=@n3, Switch=@n3, "PC-Low(NotPerf)&&Medium"=@n3) {materialFunction=&PlatformSwitch_1D}.Result) {atlas=&/CA_TEMP_ToyFox, curve=&/CC_TEMP_ToyFoxRefraction, parameterName=Atlas}
  LinearInterpolate(
   B=LinearInterpolate(
    A=ScalarParameter() {defaultValue=1, parameterName=Refract Low}
    B=ScalarParameter() {defaultValue=1.4, parameterName=Refract High}
    Alpha=NamedRerouteDeclaration(None=@n6.R) {name=CurveRefraction, nodeColor=<rgba 0 1 0.0859 1>}
   )
   Alpha=ParticleRandom()
  ) {constA=1}
  n7 = TextureCoordinate()
  n8 = VectorParameter() {parameterName=Normal Texture options, channelNames {a="", b=Flatness, g=Tile Y, r=tile X}, defaultValue=<rgba 1 1 0 0>}
  n9 = ComponentMask(None=@n8) {g=true, r=true}
  MaterialFunctionCall(
   Normal=TextureSampleParameter2D(UVs=Multiply(A=@n7, B=@n9)) {parameterName=Normal, samplerType=SAMPLERTYPE_Normal, texture=&FlatNormalBlue}.RGB
   Flatness=@n8.Flatness
  ) {materialFunction=&FlattenNormal}
  MaterialFunctionCall() {materialFunction=&CameraDepthFade}
  NamedRerouteDeclaration(None=@n6.A) {name=CurveNA, nodeColor=<rgba 1 1 1 1>}
  n10 = ParticleColor()
  n11 = NamedRerouteDeclaration(None=@n6.G) {name=CurveOpacity, nodeColor=<rgba 0 1 0 1>}
  n12 = ScalarParameter() {defaultValue=1, parameterName=CUre Opacity Str}
  Reroute(None=Multiply(A=@n10.A, B=Multiply(A=@n11, B=@n12)))
  Reroute(None=Multiply(A=@n10.RGB, B=NamedRerouteDeclaration(None=@n6.B) {name=CurveEmiisve Str, nodeColor=<rgba 0 0 1 1>}))
 ]
}
