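# Temporary additive, unlit, two-sided material for Niagara mesh particles, named for a ToyFox sphere effect.
# Inverts and sharpens a depth fade to isolate geometry intersections, scaled by the "Add Inverted Depth Fade" parameter.
# Separate chains add that intersection term to particle color and alpha, multiply color by inverted Fresnel, and multiply alpha by plain depth fade.
# No links from these chains to material outputs are shown.
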Material /ToyFoxGameplay/Assets/FX/TEMP/M_TEMP_ToyFox_Sphere
graph {
 properties {
  bUsedWithNiagaraMeshParticles=true
  blendMode=BLEND_Additive
  pixelDepthOffsetMode=PDOM_Legacy
  shadingModel=MSM_Unlit
  twoSided=true
  usedShadingModels=Unlit
 }
 nodes [
  n1 = ParticleColor()
  n2 = DepthFade()
  n3 = OneMinus(None=Power(Base=@n2) {constExponent=8})
  n4 = Multiply(A=@n3, B=ScalarParameter() {defaultValue=2, parameterName=Add Inverted Depth Fade})
  Add(A=Add(A=@n1.A, B=@n3), B=@n4)
  Add(A=@n1.RGB, B=Multiply(A=@n1.RGB, B=@n4))
  Multiply(A=OneMinus(None=Fresnel() {baseReflectFraction=0.5, exponent=0.1}), B=@n1.RGB)
  Multiply(A=@n1.A, B=@n2)
 ]
}
