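# Temporary translucent unlit ToyFox FX material for lightning sparks on Niagara mesh particles, drawing billboarded sprites with a soft circular mask.
# A Sprite function builds a position offset, sized between MinSize and MaxSize by red vertex color, scaled by dynamic Shrink_Speed and near-camera fade.
# A vertex-normal offset driven by dynamic Spreed_Speed is added to spread the sparks.
# Particle color passes through MF_FX_EmissiveAdjustments, alongside a UV SphereMask.
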Material /ToyFoxGameplay/Assets/FX/TEMP/Materials/M_TEMP_ToyFox_Lightning_Sparks_Mesh
graph {
 properties {
  bCastRayTracedShadows=false
  bUsedWithInstancedStaticMeshes=true
  bUsedWithMeshParticles=true
  bUsedWithNiagaraMeshParticles=true
  blendMode=BLEND_Translucent
  pixelDepthOffsetMode=PDOM_Legacy
  shadingModel=MSM_Unlit
  usedShadingModels=Unlit
 }
 nodes [
  n1 = WorldPosition()
  n2 = DynamicParameter() {defaultValue=<rgba 1 1 1 0>, paramNames [Spreed_Speed, Shrink_Speed, "", ""]}
  n3 = Multiply(
   A=@n2.Shrink_Speed
   B=LinearInterpolate(
    A=ScalarParameter() {defaultValue=85, parameterName=MinSize}
    B=ScalarParameter() {defaultValue=300, parameterName=MaxSize}
    Alpha=VertexColor().R
   )
  )
  n4 = ScalarParameter() {defaultValue=10000, group=WPO, parameterName=NearCamShrinkDist}
  n5 = Constant3Vector() {constant=<rgba 100 100 100 0>}
  n6 = VertexNormalWS()
  Add(
   A=Multiply(
    A=MaterialFunctionCall!(Center Location=@n1.XYZ, 0-1 UVs, XY scale=@n3, OverridePivot, "Normals (Optional)") {materialFunction=&Sprite}.World position offset
    B=MaterialFunctionCall(Distance=@n4) {materialFunction=&Near_Camera_Fade}.Result
   )
   B=Multiply(A=Multiply(A=@n5, B=@n6) {constA=1}, B=@n2.Spreed_Speed)
  )
  Constant2Vector() {g=1000, r=1000}
  MaterialFunctionCall(Emissive=ParticleColor().RGB) {materialFunction=&MF_FX_EmissiveAdjustments}
  n7 = Constant() {r=0.5}
  SphereMask(A=TextureCoordinate(), B=@n7, Radius=Constant() {r=0.35}, Hardness=@n7)
 ]
}
