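# Temporary translucent unlit Niagara beam material for ToyFox, usable on sprites and mesh particles.
# Spline-thickened UVs, switchable to mesh UVs, tile a panning beam texture distorted by noise, with a mask varying thickness at the mouth.
# Smoothstep and detail-noise erosion shape opacity alongside camera near fade; core and detail masks tint hue-shifted particle color.
# Dynamic parameters drive size, erosion, hue and end position, whose tip mask reads X in place of Y.
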Material /ToyFoxGameplay/Assets/FX/TEMP/Materials/TEMP_M_FX_ToyFoxBeamSpline
graph {
 properties {
  bAutomaticallySetUsageInEditor=false
  bUsedWithNiagaraMeshParticles=true
  bUsedWithNiagaraSprites=true
  blendMode=BLEND_Translucent
  shadingModel=MSM_Unlit
  usedShadingModels=Unlit
 }
 nodes [
  n1 = NamedRerouteUsage()
  n2 = VectorParameter() {group=Detail, parameterName=LightElementsColor, defaultValue=<rgba 0 0 0 1>}
  n3 = OneMinus(None=Multiply(A=NamedRerouteUsage(), B=ScalarParameter() {group=Detail, parameterName=DarknessContribution}))
  n4 = TextureCoordinate()
  n5 = ParticleColor()
  n6 = NamedRerouteUsage()
  n7 = VectorParameter() {group=Core, parameterName=CoreColor, defaultValue=<rgba 0 0 0 1>}
  n8 = MaterialFunctionCall(
   Hue Shift Percentage=Multiply(
    A=ComponentMask(None=@n4) {r=true}
    B=DynamicParameter() {parameterIndex=3, paramNames [BeamLength/CameraNearFade, BeamWidth/Tiling/, Param3//, Param4]}.Param4
   )
   Texture=LinearInterpolate(A=@n5.RGB, B=Multiply(A=@n6, B=@n7.RGB), Alpha=@n6)
  ) {materialFunction=&HueShift}
  MaterialFunctionCall(Emissive=Add(A=Multiply(A=@n1, B=@n2.RGB), B=Multiply(A=@n3, B=@n8.Result))) {materialFunction=&MF_FX_EmissiveAdjustments}
  n9 = SphereMask(
   A=CameraPositionWS()
   B=WorldPosition().XYZ
   Radius=DynamicParameter() {parameterIndex=1, paramNames [BeamLength/CameraNearFade, BeamWidth/Tiling/, Param3//, UseSplineUV]}.BeamLength/CameraNearFade
  ) {hardnessPercent=35}
  n10 = NamedRerouteUsage()
  Multiply(
   A=NamedRerouteDeclaration(None=OneMinus(None=@n9)) {name=NearCameraFade, nodeColor=<rgba 1 0 0.1875 1>}
   B=Saturate(None=Multiply(A=@n5.A, B=@n10))
  )
  NamedRerouteDeclaration(None=DynamicParameter() {paramNames [BeamLength/CameraNearFade, BeamWidth/Tiling/, ErodeIn3//, ErodeOut]}.BeamLength/CameraNearFade) {name=BeamLength, nodeColor=<rgba 0.9453 1 0 1>}
  n11 = NamedRerouteUsage()
  n12 = NamedRerouteUsage()
  n13 = VectorParameter() {group=Noise, parameterName=Noise_Scale, defaultValue=<rgba 0 0 0 1>}
  n14 = Multiply(A=ComponentMask(None=@n13.RGB) {g=true, r=true}, B=@n13.B)
  n15 = VectorParameter() {group=Noise, parameterName=Noise_Speed, defaultValue=<rgba 0 0 0 1>}
  n16 = TextureSample(UVs=Panner(Coordinate=Multiply(A=@n12, B=@n14), Speed=ComponentMask(None=@n15.RGB) {g=true, r=true}) {bFractionalPart=true, speedX=-1}) {texture=/Game/Effects/Fort_Effects/Textures/Fire/T_FireNoise_02}
  n17 = ComponentMask(None=Subtract(A=Multiply(A=@n16.RGB) {constB=2})) {r=true}
  n18 = ComponentMask(None=VectorParameter() {group=Noise, parameterName=Noise_DistortMain, defaultValue=<rgba 0 0 0 1>}.RGB) {g=true, r=true}
  n19 = TextureSample(UVs=Add(A=Panner(Coordinate=@n11) {bFractionalPart=true, speedX=-4.5}, B=Multiply(A=@n17, B=@n18))) {samplerType=SAMPLERTYPE_LinearGrayscale, texture=/ToyFoxGameplay/Assets/FX/TEMP/Textures/TEMP_T_FX_ToyFox_Beam}
  n20 = Add(A=@n19.R, B=ScalarParameter() {group=Opacity, parameterName=Opacity_Offset})
  n21 = Add(A=@n20, B=ScalarParameter() {group=Opacity, parameterName=Opacity_Min})
  n22 = Add(A=@n20, B=ScalarParameter() {group=Opacity, parameterName=Opacity_Max})
  n23 = Multiply(A=@n16.G, B=ScalarParameter() {group=Noise, parameterName=Noise_Opacity_Contribution})
  n24 = ScalarParameter() {defaultValue=1, group=Erosion, parameterName=ValueForFullErosion}
  n25 = DynamicParameter() {parameterIndex=2, paramNames [BeamLength/CameraNearFade/EndPosX, BeamWidth/Tiling/EndPosY, Param3//EndPosZ, Param4]}
  n26 = SphereMask(
   A=WorldPosition().XYZ
   B=AppendVector(A=AppendVector(A=@n25.BeamLength/CameraNearFade/EndPosX, B=@n25.BeamLength/CameraNearFade/EndPosX), B=@n25.Param3//EndPosZ)
   Radius=ScalarParameter() {group=Erosion, parameterName=TipErode}
  ) {hardnessPercent=0}
  n27 = Add(A=Multiply(A=Multiply(A=@n23, B=@n19.R), B=LinearInterpolate(B=@n24, Alpha=@n26) {constA=1}), B=Constant())
  n28 = DynamicParameter() {paramNames [BeamLength/CameraNearFade, BeamWidth/Tiling/, ErodeIn3//, ErodeOut]}
  n29 = NamedRerouteUsage()
  n30 = VectorParameter() {group=Detail, parameterName=Detail_Scale, defaultValue=<rgba 0 0 0 1>}
  n31 = Multiply(A=ComponentMask(None=@n30.RGB) {g=true, r=true}, B=@n30.B)
  n32 = VectorParameter() {group=Detail, parameterName=Detail_Speed, defaultValue=<rgba 0 0 0 1>}
  n33 = ComponentMask(None=@n32.RGB) {g=true, r=true}
  n34 = TextureSample(UVs=Panner(Coordinate=Multiply(A=@n29, B=@n31), Speed=Multiply(A=@n33, B=@n32.B)) {bFractionalPart=true, speedX=-1}) {texture=/Game/Effects/Fort_Effects/Textures/Fire/HF_FireNoisesPacked}
  n35 = Subtract(A=@n28.ErodeIn3//, B=@n34.B)
  NamedRerouteDeclaration(None=Multiply(A=SmoothStep(Min=@n21, Max=@n22, Value=@n27), B=Saturate(None=@n35))) {name=BeamOpacity, nodeColor=<rgba 1 0.1641 0 1>}
  n36 = SphereMask(
   A=@n19.R
   B=ScalarParameter() {group=CoreColoring, parameterName=Core_Pos}
   Radius=ScalarParameter() {group=CoreColoring, parameterName=Core_Width}
   Hardness=ScalarParameter() {group=CoreColoring, parameterName=Core_Hardness}
  )
  n37 = ScalarParameter() {group=CoreColoring, parameterName=Core_Exponent}
  NamedRerouteDeclaration(None=Saturate(None=Power(Base=@n36, Exp=@n37))) {name=CoreMask, nodeColor=<rgba 1 0 0.0234 1>}
  n38 = SmoothStep(
   Min=AppendVector(A=ScalarParameter() {group=Detail, parameterName=Detail_Dark_Min}, B=ScalarParameter() {group=Detail, parameterName=Detail_Light_Min})
   Max=AppendVector(A=ScalarParameter() {group=Detail, parameterName=Detail_Dark_Max}, B=ScalarParameter() {group=Detail, parameterName=Detail_Light_Max})
   Value=AppendVector(A=@n34.R, B=@n34.R)
  )
  NamedRerouteDeclaration(None=ComponentMask(None=@n38) {r=true}) {name=Detail_DarkElements, nodeColor=<rgba 1 0 0.5859 1>}
  NamedRerouteDeclaration(None=ComponentMask(None=@n38) {g=true}) {name=Detail_LightElements, nodeColor=<rgba 1 0 0.5859 1>}
  n39 = Divide(
   A=NamedRerouteUsage()
   B=DynamicParameter() {parameterIndex=1, paramNames [BeamLength/CameraNearFade, BeamWidth/Tiling/, Param3//, UseSplineUV]}.BeamWidth/Tiling/
  )
  n40 = Constant() {r=1}
  NamedRerouteDeclaration(None=Multiply(A=NamedRerouteUsage(), B=AppendVector(A=@n39, B=@n40))) {name=BeamUV, nodeColor=<rgba 0 0.25 1 1>}
  n41 = ParticlePositionWS()
  n42 = WorldPosition()
  n43 = ScalarParameter() {defaultValue=50, parameterName=MouthMaskOut}
  NamedRerouteDeclaration(None=Power(Base=SphereMask(A=@n41, B=@n42.XYZ, Radius=@n43) {hardnessPercent=0}) {constExponent=4}) {name=MouthSourceMask, nodeColor=<rgba 1 0 0.8203 1>}
  n44 = DynamicParameter() {paramNames [BeamLength/CameraNearFade, BeamWidth/Tiling/, ErodeIn3//, ErodeOut]}
  n45 = Multiply(A=ScalarParameter() {parameterName=Thickness_End}, B=@n44.BeamWidth/Tiling/)
  n46 = Multiply(
   A=LinearInterpolate(A=ScalarParameter() {parameterName=Thickness_Base}, B=ScalarParameter() {parameterName=Thickness_Mouth}, Alpha=NamedRerouteUsage())
   B=@n44.BeamWidth/Tiling/
  )
  n47 = StaticBool()
  NamedRerouteDeclaration(None=LinearInterpolate(
    A=TextureCoordinate()
    B=MaterialFunctionCall(WidthBase=@n45, WidthTip=@n46, Expand U or V UV Channel=@n47) {materialFunction=&SplineThicken}.UVs with Parallax
    Alpha=DynamicParameter() {parameterIndex=1, paramNames [BeamLength/CameraNearFade, BeamWidth/Tiling/, Param3//, UseSplineUV]}.UseSplineUV
   )) {name=Spline_UV, nodeColor=<rgba 0 0.9766 1 1>}
  ScalarParameter() {defaultValue=0.5, group=InOut, parameterName=Erosion_P}
  Subtract(A=@n28.ErodeIn3//, B=ComponentMask(None=TextureCoordinate()) {r=true})
 ]
}
