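# Temporary unlit additive material for Niagara ribbons, named as a glow tube effect for ToyFox.
# An AxisAlignedFresnel along the transformed local X axis, shaped by an Exponent parameter, is multiplied by particle alpha.
# A plain Fresnel node and an inverted Fresnel function sharing that Exponent are left unused.
# The graph shows no link from the final multiply to a material output, so it appears unfinished.
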Material /ToyFoxGameplay/Assets/FX/TEMP/Materials/TEMP_M_FX_ToyFoxRibbonGlowTube
graph {
 properties {
  bAutomaticallySetUsageInEditor=false
  bUsedWithNiagaraRibbons=true
  blendMode=BLEND_Additive
  shadingModel=MSM_Unlit
  usedShadingModels=Unlit
 }
 nodes [
  Fresnel()
  n1 = ScalarParameter() {parameterName=Exponent}
  MaterialFunctionCall(Invert Fresnel=StaticBool() {value=true}, Power=@n1) {materialFunction=&Fresnel_Function}
  n2 = Constant3Vector() {constant=<rgba 1 0 0 0>}
  Multiply(
   A=ParticleColor().A
   B=MaterialFunctionCall(Axis=Transform(None=@n2) {transformSourceType=TRANSFORMSOURCE_Local}, Exponent=@n1) {materialFunction=&AxisAlignedFresnel}.Fresnel
  )
 ]
}
