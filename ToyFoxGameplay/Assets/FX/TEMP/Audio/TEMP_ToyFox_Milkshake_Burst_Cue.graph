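# Temporary Sound Cue for the ToyFox milkshake explosion burst effect.
# A single Wave Player feeds the root directly with one burst wave, with no randomization or mixing.
# It uses its own burst attenuation asset and the Fort_Events_Scripted sound class.
# A concurrency override allows only one instance at a time, with a short retrigger time to limit rapid repeats.
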SoundCue /ToyFoxGameplay/Assets/FX/TEMP/Audio/TEMP_ToyFox_Milkshake_Burst_Cue
root=WavePlayer_1
nodes [{id=WavePlayer_1, properties {soundWaveAssetPtr=/ToyFoxGameplay/Assets/FX/TEMP/Audio/wavs/TEMP_ToyFox_Milkshake_Expl_Burst_01}}]
properties {
 attenuationSettings=/ToyFoxGameplay/Assets/FX/TEMP/Audio/Settings/ATT_TEMP_ToyFox_Milkshake_Burst
 bOverrideConcurrency=true
 maxDistance=2097152
 soundClassObject=/Game/Sounds/FortSoundClassesAndModes/Fort_Events_Scripted
 volumeMultiplier=1
 concurrencyOverrides {maxCount=1, retriggerTime=0.1}
}
