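# Temporary Niagara burst effect for ToyFox "Milkshake": an explosive eruption of smoke, dust spikes, dirt and rocks, with a shockwave mesh that plays the burst sound.
# When User.FirstBurst is true, chunks, smoke and streaks also spawn along five cracks placed by user position and orientation arrays, with curves shaping the smoke paths.
# A Time-of-Day parameter tints dust and smoke, and distance scalability trims cost.
# DirectionalBurstPaint and Directional_Arm are disabled.
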NiagaraSystem /ToyFoxGameplay/Assets/FX/TEMP/NS_TEMP_ToyFox_Milkshake_Burst
dependencies {
 usedDataInterfaces [
  &NiagaraDataInterfaceCurve
  &NiagaraDataInterfaceColorCurve
  &NiagaraDataInterfaceSpriteRendererInfo
  &NiagaraDataInterfaceMeshRendererInfo
  &NiagaraDataInterfaceVector2DCurve
  &NiagaraDataInterfaceAudioPlayer
  &NiagaraDataInterfaceArrayFloat
 ]
 usedDynamicInputs []
 usedModules [&SystemState]
}
emitters {
 DirectionalBurstPaint {
  bEnabled=false
  simTarget=GPUComputeSim
  emitterUpdateScript [
   &EmitterState(
    Enable Distance Culling=true
    Life Cycle Mode=::NewEnumerator1
    Loop Behavior=::NewEnumerator1
    Loop Duration=0.2
    MaxDistance=7500
    Module.bUseMaxDistance=true
    Module.bUseMinDistance=false
    Scalability Mode=::NewEnumerator1
    UseLoopDelay=false
   )
   &SpawnRate(SpawnRate=350, Use Spawn Probability=false)
  ]
  particleSpawnScript [
   &InitializeParticle(
    Color Maximum=<rgba 1 1 1 0.9>
    Color Minimum=<rgba 1 1 1 0.6>
    Color Mode=::NewEnumerator2
    Lifetime Max=1.2
    Lifetime Mode=::NewEnumerator1
    Sprite Size Mode=::NewEnumerator1
    Uniform Sprite Size Max=550
    Uniform Sprite Size Min=500
    UsePositionOffset=false
    Write Material Random=true
   )
   &ShapeLocation(MaskSpawnGroup=false, Offset=<0 0 -150>, Offset Mode=::NewEnumerator0, Override Seed=false, Sphere Radius=400, UseOwnerScale=true)
   &AddVelocity(
    Cone Angle=135
    Cone Axis=<0 0 1>
    Inner Cone Angle=45
    Module.UseConeFalloff=false
    Module.UseVelDistribution=false
    Override Seed=false
    Velocity Mode=::NewEnumerator2
    Velocity Speed=5
   )
  ]
  particleUpdateScript [
   &ParticleState()
   &ScaleColor(
    Scale Alpha=&FloatFromCurve(FloatCurve=DataInterface() {propertyValues {Curve {keys [<at 0.3 1 RCIM_Cubic>, {arriveTangent=-4.7596, interpMode=RCIM_Cubic, leaveTangent=-4.7596, tangentMode=RCTM_User, time=1, value=-0.1}]}}})
    Scale RGB=&MakeVectorFromLinearColorRGB(Color=&ColorFromCurve(
      ColorCurve=DataInterface() {propertyValues {
        AlphaCurve {keys [<at 0 1 RCIM_Cubic>]}
        BlueCurve {keys [<at 0 0.167 RCIM_Cubic>,<at 0.25 0.0719>,<at 0.5 0.0775>,<at 0.75 0.0262>,<at 1 0.167 RCIM_Cubic>]}
        GreenCurve {keys [<at 0 0.2032 RCIM_Cubic>,<at 0.25 0.2239>,<at 0.5 0.2212>,<at 0.75 0.1317>,<at 1 0.2032 RCIM_Cubic>]}
        RedCurve {keys [<at 0 0.2812 RCIM_Cubic>,<at 0.25 0.5521>,<at 0.5 0.5312>,<at 0.75 0.3594>,<at 1 0.2812 RCIM_Cubic>]}
       }}
      CurveIndex=Linked() {linkedVariable {name=NPC.FortniteNPC.TimeOfDay_0-1, type=NiagaraFloat}}
     ))
    ScaleA=true
    ScaleRGB=true
   )
   &SolveForcesAndVelocity(Clamp Velocity=false, Limit Acceleration=false)
   &ScaleSpriteSize(
    Non-Uniform Curve Scale=<0.75 1.75>
    Non-Uniform Curve Sprite Scale=DataInterface() {propertyValues {
      XCurve {keys [
        {arriveTangent=4.1785, interpMode=RCIM_Cubic, leaveTangent=4.1785, tangentMode=RCTM_User, time=0, value=0.5}
        {arriveTangent=0.0224, interpMode=RCIM_Cubic, leaveTangent=0.0224, tangentMode=RCTM_User, time=1, value=2.5}
       ]}
      YCurve {keys [
        {arriveTangent=4.1785, interpMode=RCIM_Cubic, leaveTangent=4.1785, tangentMode=RCTM_User, time=0, value=1}
        {arriveTangent=0.0224, interpMode=RCIM_Cubic, leaveTangent=0.0224, tangentMode=RCTM_User, time=1, value=3}
       ]}
     }}
    Scale Sprite Size Mode=::NewEnumerator3
   )
  ]
  rendererClasses [&NiagaraSpriteRendererProperties]
  renderers [{rendererClass=&NiagaraSpriteRendererProperties}]
 }
 Directional_Arm {
  bEnabled=false
  simTarget=GPUComputeSim
  emitterUpdateScript [
   &EmitterState()
   &SpawnBurst_Instantaneous(
    Spawn Count=&RandomRangeInt(Minimum=3, Maximum=5)
    Spawn Group=1
    Use Spawn Probability=false
    UseLoopCountLimit=false
   )
  ]
  particleSpawnScript [
   &InitializeParticle(
    Color=&RandomRangeLinearColor(Minimum=<rgba 0.42 0.2182 0.1247 0.35>, Maximum=<rgba 0.45 0.2226 0.1172 0.45>)
    Lifetime Max=2.5
    Lifetime Mode=::NewEnumerator1
    Mesh Scale Mode=::NewEnumerator1
    Mesh Uniform Scale Max=22
    Mesh Uniform Scale Min=18
    UsePositionOffset=false
    Write Material Random=true
   )
   &ShapeLocation(MaskSpawnGroup=false, Override Seed=false, Ring Radius=500, Shape Primitive=::NewEnumerator4, UseOwnerScale=true)
   &AddVelocity(Module.Use Falloff Distance=false, Origin Offset=<0 0 -750>, Override Seed=false, Velocity Mode=::NewEnumerator1, Velocity Speed=10)
  ]
  particleUpdateScript [
   &ParticleState()
   &SolveForcesAndVelocity(Clamp Velocity=false, Limit Acceleration=false)
   &DynamicMaterialParameters(
    Index 0 Param 1=0.25
    Index 0 Param 2=0.5
    Index 0 Param 3=&FloatFromCurve(
     FloatCurve=DataInterface() {propertyValues {Curve {keys [
         {arriveTangent=3.1711, interpMode=RCIM_Cubic, leaveTangent=3.1711, tangentMode=RCTM_User, time=0, value=0.2}
         {arriveTangent=-0.0405, interpMode=RCIM_Cubic, leaveTangent=-0.0405, tangentMode=RCTM_User, time=0.5, value=1}
        ]}}}
     Scale Curve=&RandomRangeFloat(Minimum=16, Maximum=20)
    )
    Index 0 Param 4=&FloatFromCurve(FloatCurve=DataInterface() {propertyValues {Curve {keys [
         {interpMode=RCIM_Cubic, tangentMode=RCTM_User, time=0, value=1}
         {interpMode=RCIM_Cubic, tangentMode=RCTM_User, time=0.1, value=0}
         {arriveTangent=3.5202, tangentMode=RCTM_User, time=1, value=1}
        ]}}})
    Module.Param0WriteEnabled=true
    Module.Param1WriteEnabled=true
    Module.Param2WriteEnabled=true
    Module.Param3WriteEnabled=true
   )
   &ScaleMeshSize(Scale Factor=&VectorFromFloat(Value=&FloatFromCurve(FloatCurve=DataInterface() {propertyValues {Curve {keys [{arriveTangent=2.4896, interpMode=RCIM_Cubic, leaveTangent=2.4896, tangentMode=RCTM_User, time=0, value=0}, <at 1 2 RCIM_Cubic>]}}})))
   &CameraOffset(Camera Offset Amount=-25)
  ]
  rendererClasses [&NiagaraMeshRendererProperties]
  renderers [{rendererClass=&NiagaraMeshRendererProperties, rendererIndex=0}]
 }
 Dirt {
  emitterUpdateScript [
   &EmitterState(
    Enable Distance Culling=true
    MaxDistance=7500
    Module.bApplySpawnCountScaleByDistance=true
    Module.bUseMaxDistance=true
    Module.bUseMinDistance=false
    Scalability Mode=::NewEnumerator1
    Scale Spawn Count=true
    Spawn Count Scale By Distance=DataInterface() {propertyValues {Curve {keys [<at 0 1>,<at 0.5 1>,<at 1 0.1>]}}}
   )
   &SpawnBurst_Instantaneous(Spawn Count=250, Use Spawn Probability=false, UseLoopCountLimit=false)
  ]
  particleSpawnScript [
   &InitializeParticle(
    Color Minimum=<rgba 1 1 1 0.6>
    Color Mode=::NewEnumerator2
    Lifetime Max=2.5
    Lifetime Min=2
    Lifetime Mode=::NewEnumerator1
    Sprite Rotation Mode=::NewEnumerator2
    Sprite Size Mode=::NewEnumerator1
    Uniform Sprite Size Max=200
    Uniform Sprite Size Min=100
    UsePositionOffset=false
    Write Material Random=true
   )
   &ShapeLocation(
    Disc Coverage=1
    MaskSpawnGroup=false
    Override Seed=false
    Ring Radius=750
    Rotation Angle=0.3
    Rotation Axis=<0 0 1>
    Rotation Mode=::NewEnumerator1
    Shape Primitive=::NewEnumerator4
    UseOwnerScale=true
   )
   &AddVelocityFromPoint(
    Module.Use Falloff Distance=false
    Origin Offset=<0 0 -500>
    Velocity Strength=&RandomRangeFloat(Minimum=100, Maximum=1400)
   )
   &AddVelocity(Velocity=&RandomRangeVector(Minimum=<0 0 1150>, Maximum=<0 0 1687.5>))
  ]
  particleUpdateScript [
   &ParticleState()
   &GravityForce(Gravity=&RandomRangeVector(Minimum=<0 0 -1500>, Maximum=<0 0 -1875>))
   &Drag(Drag=&RandomRangeFloat(Minimum=0.5, Maximum=1), Use Linear Drag=true, Use Rotational Drag=true)
   &ScaleColor(
    Scale Alpha=&FloatFromCurve(FloatCurve=DataInterface() {propertyValues {Curve {keys [<at 0.3 1 RCIM_Cubic>, {arriveTangent=-4.7596, interpMode=RCIM_Cubic, leaveTangent=-4.7596, tangentMode=RCTM_User, time=1, value=-0.1}]}}})
    Scale RGB=&MakeVectorFromLinearColorRGB(Color=&ColorFromCurve(
      ColorCurve=DataInterface() {propertyValues {
        AlphaCurve {keys [<at 0 1 RCIM_Cubic>]}
        BlueCurve {keys [<at 0 0.167 RCIM_Cubic>,<at 0.25 0.0719>,<at 0.5 0.0896>,<at 0.75 0.0262>,<at 1 0.167 RCIM_Cubic>]}
        GreenCurve {keys [<at 0 0.2032 RCIM_Cubic>,<at 0.25 0.2239>,<at 0.5 0.2559>,<at 0.75 0.1317>,<at 1 0.2032 RCIM_Cubic>]}
        RedCurve {keys [<at 0 0.2812 RCIM_Cubic>,<at 0.25 0.5521>,<at 0.5 0.6146>,<at 0.75 0.3594>,<at 1 0.2812 RCIM_Cubic>]}
       }}
      CurveIndex=Linked() {linkedVariable {name=NPC.FortniteNPC.TimeOfDay_0-1, type=NiagaraFloat}}
     ))
    ScaleA=true
    ScaleRGB=true
   )
   &SolveForcesAndVelocity(Clamp Velocity=false, Limit Acceleration=false)
   &ScaleSpriteSize(Scale Factor=&Vector2DFromFloat(Value=&FloatFromCurve(FloatCurve=DataInterface() {propertyValues {Curve {keys [
          {interpMode=RCIM_Cubic, leaveTangent=4.8701, tangentMode=RCTM_Break, time=0, value=1}
          {interpMode=RCIM_Cubic, tangentMode=RCTM_Break, time=1, value=2.5}
         ]}}})))
   &SpriteRotationRate(Rotation Rate=45)
  ]
  rendererClasses [&NiagaraSpriteRendererProperties]
  renderers [{rendererClass=&NiagaraSpriteRendererProperties}]
 }
 DustSpikes {
  emitterUpdateScript [
   &EmitterState(
    Enable Distance Culling=true
    MaxDistance=250000
    Min Distance Response=::SleepAndLetParticlesFinish
    MinDistance=9000
    Module.bUseMaxDistance=true
    Module.bUseMinDistance=true
    Scalability Mode=::NewEnumerator1
   )
   &SpawnBurst_Instantaneous(Spawn Count=3, Use Spawn Probability=false, UseLoopCountLimit=false)
  ]
  particleSpawnScript [
   &InitializeParticle(
    Color Maximum=<rgba 0.5156 0.2335 0.1343 1>
    Color Minimum=<rgba 0.5 0.2187 0.1198 1>
    Color Mode=::NewEnumerator2
    Lifetime=1.5
    Mesh Scale Mode=::NewEnumerator1
    Mesh Uniform Scale Max=65
    Mesh Uniform Scale Min=55
    UsePositionOffset=false
    Write Material Random=true
   )
   &ShapeLocation(MaskSpawnGroup=false, Override Seed=false, Sphere Radius=10, UseOwnerScale=true)
   &AddVelocity(
    Cone Angle=90
    Cone Axis=<0 0 1>
    Module.UseConeFalloff=false
    Module.UseVelDistribution=false
    Override Seed=true
    Random Seed=6
    Velocity Mode=::NewEnumerator2
    Velocity Speed=1
   )
  ]
  particleUpdateScript [
   &ParticleState()
   &SolveForcesAndVelocity(Clamp Velocity=false, Limit Acceleration=false)
   &DynamicMaterialParameters(
    Index 0 Param 1=&FloatFromCurve(FloatCurve=DataInterface() {propertyValues {Curve {keys [{interpMode=RCIM_Cubic, tangentMode=RCTM_User, time=0, value=0.35}, {interpMode=RCIM_Cubic, tangentMode=RCTM_User, time=1, value=1}]}}})
    Index 0 Param 2=0.5
    Index 0 Param 3=&RandomRangeFloat(Minimum=0, Maximum=2)
    Index 0 Param 4=&FloatFromCurve(
     FloatCurve=DataInterface() {propertyValues {Curve {keys [
         {arriveTangent=1.1755, interpMode=RCIM_Cubic, leaveTangent=1.1755, tangentMode=RCTM_User, time=0, value=0.2}
         {arriveTangent=0.2667, interpMode=RCIM_Cubic, leaveTangent=0.2667, tangentMode=RCTM_User, time=1, value=1}
        ]}}}
     Scale Curve=&RandomRangeFloat(Minimum=125, Maximum=150)
    )
    Module.Index0Param0Write=true
    Module.Index0Param1Write=true
    Module.Index0Param2Write=true
    Module.Index0Param3Write=true
   )
   &ScaleMeshSize(Scale Factor=&VectorFromFloat(Value=&FloatFromCurve(FloatCurve=DataInterface() {propertyValues {Curve {keys [
          {arriveTangent=1.4777, interpMode=RCIM_Cubic, leaveTangent=1.4777, tangentMode=RCTM_User, time=0, value=0}
          {arriveTangent=0.5271, interpMode=RCIM_Cubic, leaveTangent=0.5271, tangentMode=RCTM_User, time=1, value=1}
         ]}}})))
   &ScaleColor(
    Scale Alpha=&FloatFromCurve(FloatCurve=DataInterface() {propertyValues {Curve {keys [
         {tangentMode=RCTM_User, time=0, value=1}
         {arriveTangent=-1, interpMode=RCIM_Cubic, leaveTangent=-1, time=0.5, value=1}
         {interpMode=RCIM_Cubic, leaveTangent=-8.0089, time=1, value=0}
        ]}}})
    Scale RGB=&Multiply_VectorByFloat(
     Vector=&MakeVectorFromLinearColorRGB(Color=&ColorFromCurve(
       ColorCurve=DataInterface() {propertyValues {
         AlphaCurve {keys [<at 0 1 RCIM_Cubic>]}
         BlueCurve {keys [<at 0.061 0.8594 RCIM_Cubic>,<at 0.25 0.5025>,<at 0.5 0.6927>,<at 0.75 0.354>,<at 0.8984 0.8594 RCIM_Cubic>]}
         GreenCurve {keys [<at 0.061 0.4819 RCIM_Cubic>,<at 0.25 0.6648>,<at 0.5 0.7951>,<at 0.75 0.474>,<at 0.8984 0.4819 RCIM_Cubic>]}
         RedCurve {keys [<at 0.061 0.3939 RCIM_Cubic>,<at 0.25 1>,<at 0.5 1>,<at 0.75 0.7812>,<at 0.8984 0.3939 RCIM_Cubic>]}
        }}
       CurveIndex=Linked() {linkedVariable {name=NPC.FortniteNPC.TimeOfDay_0-1, type=NiagaraFloat}}
      ))
     Float=&FloatFromCurve(FloatCurve=DataInterface() {propertyValues {Curve {keys [{arriveTangent=-34.3347, interpMode=RCIM_Cubic, leaveTangent=-34.3347, tangentMode=RCTM_User, time=0, value=5}, <at 0.25 1>]}}})
    )
    ScaleA=true
    ScaleRGB=true
   )
  ]
  rendererClasses [&NiagaraMeshRendererProperties]
  renderers [{rendererClass=&NiagaraMeshRendererProperties, rendererIndex=0}]
 }
 MeshChunks {
  emitterUpdateScript [
   &EmitterState(
    Enable Distance Culling=true
    MaxDistance=7500
    Module.bApplySpawnCountScaleByDistance=true
    Module.bUseMaxDistance=true
    Module.bUseMinDistance=false
    Scalability Mode=::NewEnumerator1
    Scale Spawn Count=true
    Spawn Count Scale By Distance=DataInterface() {propertyValues {Curve {keys [<at 0 1>,<at 0.5 1>,<at 1 0.25>]}}}
   )
   &SpawnBurst_Instantaneous(
    Spawn Count=&MakeCustomIntFromBool(Bool=$User.FirstBurst, True INT=100, False INT=0)
    Use Spawn Probability=false
    UseLoopCountLimit=false
   )
  ]
  particleSpawnScript [
   &FN_SetSpawnGroup(Number of Spawn Groups=5)
   &InitializeParticle(
    Color=<rgba 0.025 0.025 0.025 1>
    Lifetime Mode=::NewEnumerator1
    Mesh Scale=&Multiply_VectorByFloat(
     Vector=&UniformRangedVector(Minimum=<0.15 0.3 0.4>, Maximum=<0.6 1 1.1>)
     Float=&UniformRangedFloat(Minimum=0.1, Maximum=0.5)
    )
    Mesh Scale Mode=::NewEnumerator4
    Position=&SelectVectorFromArray(Direct Array Index=&SpawnGroup(), Vector Selection Array=$User.Crack Positions)
    Position Mode=::NewEnumerator1
    Sprite Rotation Mode=::NewEnumerator2
    UsePositionOffset=false
    Write Material Random=true
   )
   &self:Quat_Array(Array Index:NiagaraInt32=$Particles.SpawnGroup, Quaternion Array:NiagaraDataInterfaceArrayQuat=$User.Crack Orientations) {module=Quat_Array}
   &InitialMeshOrientation(Orientation Vector=&RandomVector(Vector Scale=1), Use Orientation Vector=true, Use Rotation Vector=false)
   &ShapeLocation(
    Cylinder Height=&SelectFloatFromArray(
     Direct Array Index=&SpawnGroup()
     Float Selection Array=DataInterface() {propertyValues {FloatData [2750,3450,3400,3600,2250]}}
    )
    Cylinder Height Midpoint=0
    Cylinder Radius=250
    Hemicircle Y=true
    MaskSpawnGroup=false
    Offset=&RotateVectorByQuaternion(Vector To Rotate=<0 0 10>, Quaternion=$Particles.Quat)
    Offset Mode=::NewEnumerator0
    Override Seed=false
    Rotation Axis=&RotateVectorByQuaternion(Vector To Rotate=<1 0 0>, Quaternion=$Particles.Quat)
    Rotation Mode=::NewEnumerator1
    Shape Primitive=::NewEnumerator1
    Surface Only=true
    Use Endcaps In Surface Only Mode=false
    UseOwnerScale=true
   )
   &AddVelocity(
    Velocity=&Add_Vector(A=$Particles.ShapeLocation.ShapeNormal, B=<0.25 0.25 5>)
    Velocity Speed Scale=&RandomRangeFloat(Minimum=250, Maximum=450)
   )
   &AddRotationalVelocity(Rotation Rate=&RandomVector(Vector Scale=&RandomRangeFloat(Minimum=1, Maximum=3)))
   &ApplyInitialForces()
   &DynamicMaterialParameters(Module.Index0Param0Write=true, Module.Index0Param1Write=false, Module.Index0Param2Write=false, Module.Index0Param3Write=false)
   &CameraOffset(Camera Offset Amount=5)
  ]
  particleUpdateScript [
   &ParticleState()
   &GravityForce(Gravity=<0 0 -1600>)
   &Drag(Use Linear Drag=true, Use Rotational Drag=true)
   &SolveForcesAndVelocity(Clamp Velocity=false, Limit Acceleration=false)
   &ScaleMeshSize(Scale Factor=&VectorFromFloat(Value=&FloatFromCurve(FloatCurve=DataInterface() {propertyValues {Curve {keys [<at 0.75 1 RCIM_Cubic>, {arriveTangent=-6.2711, interpMode=RCIM_Cubic, leaveTangent=-6.2711, tangentMode=RCTM_User, time=1, value=0}]}}})))
  ]
  rendererClasses [&NiagaraMeshRendererProperties]
  renderers [{rendererClass=&NiagaraMeshRendererProperties, rendererIndex=0}]
 }
 Sharp_Rocks {
  emitterUpdateScript [
   &EmitterState(
    Enable Distance Culling=true
    MaxDistance=7500
    Module.bApplySpawnCountScaleByDistance=true
    Module.bUseMaxDistance=true
    Module.bUseMinDistance=false
    Scalability Mode=::NewEnumerator1
    Scale Spawn Count=true
    Spawn Count Scale By Distance=DataInterface() {propertyValues {Curve {keys [<at 0 1>,<at 0.5 1>,<at 1 0.1>]}}}
   )
   &SpawnBurst_Instantaneous(Spawn Count=250, Use Spawn Probability=false, UseLoopCountLimit=false)
  ]
  particleSpawnScript [
   &InitializeParticle(
    Color=<rgba 0.025 0.025 0.025 1>
    Lifetime Min=0.5
    Lifetime Mode=::NewEnumerator1
    Mesh Scale=&Multiply_VectorByFloat(
     Vector=&UniformRangedVector(Minimum=<0.15 0.3 0.4>, Maximum=<0.6 1 1.1>)
     Float=&UniformRangedFloat(Minimum=0.1, Maximum=0.5)
    )
    Mesh Scale Mode=::NewEnumerator4
    Sprite Rotation Mode=::NewEnumerator2
    UsePositionOffset=false
    Write Material Random=true
   )
   &InitialMeshOrientation(Orientation Vector=&RandomVector(Vector Scale=1), Use Orientation Vector=true, Use Rotation Vector=false)
   &ShapeLocation(
    Disc Coverage=1
    MaskSpawnGroup=false
    Override Seed=false
    Ring Radius=750
    Rotation Angle=0.3
    Rotation Axis=<0 0 1>
    Rotation Mode=::NewEnumerator1
    Shape Primitive=::NewEnumerator4
    UseOwnerScale=true
   )
   &AddVelocityFromPoint(
    Module.Use Falloff Distance=false
    Origin Offset=<0 0 -500>
    Velocity Strength=&RandomRangeFloat(Minimum=100, Maximum=2250)
   )
   &AddVelocity(Velocity=&RandomRangeVector(Minimum=<0 0 1150>, Maximum=<0 0 1687.5>))
   &AddRotationalVelocity(Rotation Rate=&RandomVector(Vector Scale=&RandomRangeFloat(Minimum=1, Maximum=3)))
   &ApplyInitialForces()
   &DynamicMaterialParameters(Module.Index0Param0Write=true, Module.Index0Param1Write=false, Module.Index0Param2Write=false, Module.Index0Param3Write=false)
   &CameraOffset(Camera Offset Amount=5)
  ]
  particleUpdateScript [
   &ParticleState()
   &GravityForce(Gravity=&RandomRangeVector(Minimum=<0 0 -980>, Maximum=<0 0 -1800>))
   &Drag(Drag=&RandomRangeFloat(Minimum=0.25, Maximum=2), Use Linear Drag=true, Use Rotational Drag=true)
   &SolveForcesAndVelocity(Clamp Velocity=false, Limit Acceleration=false)
   &ScaleMeshSize(Scale Factor=&VectorFromFloat(Value=&FloatFromCurve(FloatCurve=DataInterface() {propertyValues {Curve {keys [<at 0.75 1 RCIM_Cubic>, {arriveTangent=-6.2711, interpMode=RCIM_Cubic, leaveTangent=-6.2711, tangentMode=RCTM_User, time=1, value=0}]}}})))
  ]
  rendererClasses [&NiagaraMeshRendererProperties]
  renderers [{rendererClass=&NiagaraMeshRendererProperties, rendererIndex=0}]
 }
 Smoke_SphereCenter {
  emitterUpdateScript [
   &EmitterState()
   &SpawnBurst_Instantaneous(Spawn Count=15, Use Spawn Probability=false, UseLoopCountLimit=false)
  ]
  particleSpawnScript [
   &InitializeParticle(
    Color Maximum=<rgba 0.5 0.2667 0.15 1>
    Color Minimum=<rgba 0.5 0.2667 0.15 0.9>
    Color Mode=::NewEnumerator2
    Lifetime=&RandomRangeFloat(Minimum=2, Maximum=8)
    Sprite Rotation Angle=&RandomRangeFloat(Minimum=-25, Maximum=25)
    Sprite Rotation Mode=::NewEnumerator0
    Sprite Size Mode=::NewEnumerator1
    Uniform Sprite Size Max=1600
    Uniform Sprite Size Min=1500
    UsePositionOffset=false
    Write Material Random=true
   )
   &ShapeLocation(
    Hemisphere Distribution=<1 0.5>
    MaskSpawnGroup=false
    Non Uniform Scale=<1 1 0.5>
    Offset=<0 0 -250>
    Offset Mode=::NewEnumerator0
    Override Seed=true
    Random Seed=13
    Randomness Mode:ENiagaraRandomnessMode=::NewEnumerator1
    Sphere Radius=1200
    Sphere Surface Distribution=1
    UseOwnerScale=true
   )
   &AddVelocity(
    Module.Use Falloff Distance=false
    Origin Offset=<0 0 -200>
    Override Seed=false
    Velocity Mode=::NewEnumerator1
    Velocity Speed=&RandomRangeFloat(Minimum=750, Maximum=1200)
   )
  ]
  particleUpdateScript [
   &ParticleState()
   &GravityForce(Gravity=&RandomRangeVector(Minimum=<0 0 -337.5>, Maximum=<0 0 -450>))
   &Drag(Drag=&RandomRangeFloat(Minimum=0.5, Maximum=0.7), Use Linear Drag=true, Use Rotational Drag=true)
   &SolveForcesAndVelocity(Clamp Velocity=false, Limit Acceleration=false)
   &ScaleSpriteSize(Uniform Curve Sprite Scale=DataInterface() {propertyValues {Curve {keys [
        {arriveTangent=2.4019, interpMode=RCIM_Cubic, leaveTangent=2.4019, tangentMode=RCTM_User, time=0, value=0.75}
        {arriveTangent=0.0224, interpMode=RCIM_Cubic, leaveTangent=0.0224, tangentMode=RCTM_User, time=1, value=2}
       ]}}})
   &DynamicMaterialParameters(
    Index 0 Param 1=&RandomRangeFloat(Minimum=0.1, Maximum=0.3)
    Index 0 Param 2=0.8
    Index 0 Param 3=&FloatFromCurve(
     FloatCurve=DataInterface() {propertyValues {Curve {keys [
         {tangentMode=RCTM_User, time=0, value=1}
         {arriveTangent=-0.0405, interpMode=RCIM_Cubic, leaveTangent=-0.0405, tangentMode=RCTM_User, time=0.5, value=1}
         {arriveTangent=-3.0509, interpMode=RCIM_Cubic, leaveTangent=-3.0509, tangentMode=RCTM_User, time=1, value=0.5}
        ]}}}
     CurveIndex=$Emitter.Scalability.DistanceFraction
     Scale Curve=300
    )
    Index 0 Param 4=200
    Module.Index0Param0Write=true
    Module.Index0Param1Write=true
    Module.Index0Param2Write=true
    Module.Index0Param3Write=true
   )
   &SubUVAnimation(
    Frames Per Second:NiagaraInt32=&RandomRangeInt(Minimum=8, Maximum=10)
    Playback Mode:ENiagara_FPSPlayrate=::NewEnumerator1
    Playback Scale Curve=DataInterface() {propertyValues {Curve {keys [
        {interpMode=RCIM_Cubic, tangentMode=RCTM_User, time=0.25, value=1}
        {arriveTangent=-1.2656, interpMode=RCIM_Cubic, leaveTangent=-1.2656, tangentMode=RCTM_User, time=1, value=0}
       ]}}}
    Random Start Frame=true
    "Sprite Renderer "=DataInterface() {propertyValues {SpriteRenderer=&self:Smoke_SphereCenter_0.NiagaraSpriteRendererProperties_0}}
    Start Frame Offset:NiagaraInt32=&RandomRangeInt(Minimum=0, Maximum=16)
    Start Frame Range Override=&RandomRangeInt(Minimum=0, Maximum=0)
    SubUV Animation Mode:ENiagara_SubUVLookupModeV2=::NewEnumerator3
    UseEndFrame=true
    UsePlayRateScale=false
    UseStartFrame=true
   )
   &ScaleColor(
    Scale Alpha=&FloatFromCurve(
     FloatCurve=DataInterface() {propertyValues {Curve {keys [
         {arriveTangent=0.7573, interpMode=RCIM_Cubic, leaveTangent=0.7573, tangentMode=RCTM_User, time=0, value=0}
         {arriveTangent=-0.3388, interpMode=RCIM_Cubic, leaveTangent=-0.3388, tangentMode=RCTM_User, time=0.05, value=1}
         {arriveTangent=-2.7546, interpMode=RCIM_Cubic, leaveTangent=-2.7546, tangentMode=RCTM_User, time=1, value=0}
        ]}}}
     Scale Curve=&FloatFromCurve(
      FloatCurve=DataInterface() {propertyValues {Curve {keys [
          <at 0 1>
          {arriveTangent=-0.5, interpMode=RCIM_Cubic, leaveTangent=-0.5, time=0.25, value=1}
          {interpMode=RCIM_Cubic, leaveTangent=-0.8975, time=1, value=0.5}
         ]}}}
      CurveIndex=$Emitter.Scalability.DistanceFraction
     )
    )
    Scale RGB=&MakeVectorFromLinearColorRGB(Color=&ColorFromCurve(
      ColorCurve=DataInterface() {propertyValues {
        AlphaCurve {keys [<at 0 1 RCIM_Cubic>]}
        BlueCurve {keys [<at 0.061 0.8594 RCIM_Cubic>,<at 0.25 0.5025>,<at 0.5 0.6927>,<at 0.75 0.354>,<at 0.8984 0.8594 RCIM_Cubic>]}
        GreenCurve {keys [<at 0.061 0.4819 RCIM_Cubic>,<at 0.25 0.6648>,<at 0.5 0.7951>,<at 0.75 0.474>,<at 0.8984 0.4819 RCIM_Cubic>]}
        RedCurve {keys [<at 0.061 0.3939 RCIM_Cubic>,<at 0.25 1>,<at 0.5 1>,<at 0.75 0.7812>,<at 0.8984 0.3939 RCIM_Cubic>]}
       }}
      CurveIndex=Linked() {linkedVariable {name=NPC.FortniteNPC.TimeOfDay_0-1, type=NiagaraFloat}}
     ))
    ScaleA=true
    ScaleRGB=true
   )
   &CameraOffset(Camera Offset Amount=&FloatFromCurve(
     FloatCurve=DataInterface() {propertyValues {Curve {keys [{arriveTangent=9.6041, time=0.25, value=1}, {interpMode=RCIM_Cubic, tangentMode=RCTM_User, time=1, value=5}]}}}
     CurveIndex=$Emitter.Scalability.DistanceFraction
     Scale Curve=200
    ))
  ]
  rendererClasses [&NiagaraSpriteRendererProperties]
  renderers [{rendererClass=&NiagaraSpriteRendererProperties}]
 }
 Smoke_SphereCracks {
  emitterUpdateScript [
   &EmitterState(
    Enable Distance Culling=true
    MaxDistance=25000
    Module.bApplySpawnCountScaleByDistance=true
    Module.bUseMaxDistance=true
    Module.bUseMinDistance=false
    Scalability Mode=::NewEnumerator1
    Scale Spawn Count=true
    Spawn Count Scale By Distance=DataInterface() {propertyValues {Curve {keys [<at 0 1>,<at 0.5 1>,<at 1 0.5>]}}}
   )
   &SpawnBurst_Instantaneous(
    Spawn Count=&MakeCustomIntFromBool(Bool=$User.FirstBurst, True INT=25, False INT=0)
    Use Spawn Probability=false
    UseLoopCountLimit=false
   )
  ]
  particleSpawnScript [
   &FN_SetSpawnGroup(Number of Spawn Groups=5)
   &InitializeParticle(
    Color Maximum=<rgba 0.5 0.2667 0.15 1>
    Color Minimum=<rgba 0.5 0.2667 0.15 0.9>
    Color Mode=::NewEnumerator2
    Lifetime Max=4
    Lifetime Min=2.5
    Lifetime Mode=::NewEnumerator1
    Position=&SelectVectorFromArray(Direct Array Index=&SpawnGroup(), Vector Selection Array=$User.Crack Positions)
    Position Mode=::NewEnumerator1
    Sprite Rotation Angle=&RandomRangeFloat(Minimum=-25, Maximum=25)
    Sprite Rotation Mode=::NewEnumerator0
    Sprite Size Mode=::NewEnumerator1
    Uniform Sprite Size Max=1600
    Uniform Sprite Size Min=1400
    UsePositionOffset=false
    Write Material Random=true
   )
   &self:Quat_Array(Array Index:NiagaraInt32=$Particles.SpawnGroup, Quaternion Array:NiagaraDataInterfaceArrayQuat=$User.Crack Orientations) {module=Quat_Array}
   &CameraOffset(Camera Offset Amount=200)
   &self:Crack_Position(
    Crack 1:NiagaraFloat=&FloatFromCurve(
     FloatCurve=DataInterface() {propertyValues {Curve {keys [<at 0 0>,<at 0.0984 -0.3405>,<at 0.1717 -0.8309>,<at 0.2411 -0.8305>,<at 0.2503 -0.7056>,<at 0.2569 -0.5354>,<at 0.2911 0.8069>,<at 0.3044 0.7067>,<at 0.3504 0.5514>,<at 0.4161 0.6513>,<at 0.4614 0.5012>,<at 0.5021 0.5242>,<at 0.6525 0.3294>,<at 0.73 1.1669>,<at 0.7671 -0.3512>,<at 1 -0.5>]}}}
     CurveIndex=&ReturnNormalizedExecIndex()
     Scale Curve=225
    )
    Crack 2:NiagaraFloat=&FloatFromCurve(
     FloatCurve=DataInterface() {propertyValues {Curve {keys [
         {tangentMode=RCTM_User, time=0, value=0}
         <at 0.2413 0>
         <at 0.3848 -0.5487>
         <at 0.4752 -1.2395>
         <at 0.6355 -0.9981>
         <at 0.7622 -0.2803>
         <at 0.8308 0.1716>
         <at 0.9087 0.1963>
         {tangentMode=RCTM_User, time=1, value=0}
        ]}}}
     CurveIndex=&ReturnNormalizedExecIndex()
     Scale Curve=225
    )
    Crack 3:NiagaraFloat=&FloatFromCurve(
     FloatCurve=DataInterface() {propertyValues {Curve {keys [
         {tangentMode=RCTM_User, time=0, value=0}
         <at 0.2026 0.1571>
         <at 0.2147 -0.0139>
         <at 0.3171 -0.8666>
         <at 0.4336 -0.0588>
         <at 0.4878 0.6594>
         <at 0.522 1.0184>
         <at 0.5923 0.5584>
         <at 0.6686 0.0759>
         <at 0.761 -0.3729>
         <at 0.8755 -0.07>
         {tangentMode=RCTM_User, time=1, value=0}
        ]}}}
     CurveIndex=&ReturnNormalizedExecIndex()
     Scale Curve=225
    )
    Crack 4:NiagaraFloat=&FloatFromCurve(
     FloatCurve=DataInterface() {propertyValues {Curve {keys [
         {tangentMode=RCTM_User, time=0, value=0}
         <at 0.1145 0.2435>
         <at 0.2171 0.4328>
         <at 0.2624 0.8866>
         <at 0.4133 0.3221>
         <at 0.5114 -0.7958>
         <at 0.5778 -1.4931>
         <at 0.6955 -1.15>
         <at 0.9038 -1.803>
         <at 0.9943 -1.7809>
        ]}}}
     CurveIndex=&ReturnNormalizedExecIndex()
     Scale Curve=225
    )
    Crack 5:NiagaraFloat=&FloatFromCurve(
     FloatCurve=DataInterface() {propertyValues {Curve {keys [
         {tangentMode=RCTM_User, time=0, value=0}
         <at 0.1139 1.1844>
         <at 0.3434 -0.379>
         <at 0.5805 -0.3596>
         <at 0.7111 0.1457>
         <at 0.8281 -0.2649>
         <at 0.9151 -0.3596>
         <at 1 0>
        ]}}}
     CurveIndex=&ReturnNormalizedExecIndex()
     Scale Curve=225
    )
    Index:NiagaraInt32=$Particles.SpawnGroup
   ) {module=Crack_Position}
   /SetVariables(Particles.Position=&AddVectorToPosition(
     Position=&SelectVectorFromArray(Direct Array Index=$Particles.SpawnGroup, Vector Selection Array=$User.Crack Positions)
     Vector=&RotateVectorByQuaternion(
      Vector To Rotate=&MakeVector(
       X=&Multiply_Float(
        A=&SelectFloatFromArray(
         Direct Array Index=&SpawnGroup()
         Float Selection Array=DataInterface() {propertyValues {FloatData [2750,3450,3400,3600,2250]}}
        )
        B=&ReturnNormalizedExecIndex()
       )
       Y=$Particles.Offset
       Z=200
      )
      Quaternion=$Particles.Quat
     )
    )) {isSetParameters=true}
   &AddVelocity(
    Cone Angle=180
    Cone Axis=<0 0 1>
    Inner Cone Angle=45
    Module.UseConeFalloff=false
    Module.UseVelDistribution=false
    Override Seed=false
    Rotation Coordinate Space=::World
    Velocity Mode=::NewEnumerator2
    Velocity Speed=&RandomRangeFloat(Minimum=400, Maximum=750)
   )
   &AddVelocity(Rotation Coordinate Space=::World, Velocity=<0 0 350>) {module=AddVelocity001}
  ]
  particleUpdateScript [
   &ParticleState()
   &SpriteRotationRate(Rotation Rate=&RandomRangeFloat(Minimum=-12.5, Maximum=12.5))
   &Drag(Use Linear Drag=true, Use Rotational Drag=true)
   &GravityForce(Gravity=&RandomRangeVector(Minimum=<0 0 -200>, Maximum=<0 0 -440>))
   &SolveForcesAndVelocity(Clamp Velocity=false, Limit Acceleration=false)
   &ScaleSpriteSize(Uniform Curve Sprite Scale=DataInterface() {propertyValues {Curve {keys [
        {arriveTangent=2.4019, interpMode=RCIM_Cubic, leaveTangent=2.4019, tangentMode=RCTM_User, time=0, value=0.75}
        {arriveTangent=0.0224, interpMode=RCIM_Cubic, leaveTangent=0.0224, tangentMode=RCTM_User, time=1, value=1.5}
       ]}}})
   &DynamicMaterialParameters(
    Index 0 Param 2=0.75
    Index 0 Param 3=150
    Index 0 Param 4=200
    Module.Index0Param0Write=true
    Module.Index0Param1Write=true
    Module.Index0Param2Write=true
    Module.Index0Param3Write=true
   )
   &SubUVAnimation(
    Frames Per Second:NiagaraInt32=&RandomRangeInt(Minimum=8, Maximum=10)
    Playback Mode:ENiagara_FPSPlayrate=::NewEnumerator1
    Playback Scale Curve=DataInterface() {propertyValues {Curve {keys [
        {interpMode=RCIM_Cubic, tangentMode=RCTM_User, time=0.25, value=1}
        {arriveTangent=-1.2656, interpMode=RCIM_Cubic, leaveTangent=-1.2656, tangentMode=RCTM_User, time=1, value=0}
       ]}}}
    Random Start Frame=true
    "Sprite Renderer "=DataInterface() {propertyValues {SpriteRenderer=&self:Smoke_SphereCracks_1.NiagaraSpriteRendererProperties_0}}
    Start Frame Offset:NiagaraInt32=&RandomRangeInt(Minimum=0, Maximum=16)
    SubUV Animation Mode:ENiagara_SubUVLookupModeV2=::NewEnumerator3
    UseEndFrame=true
    UsePlayRateScale=false
    UseStartFrame=true
   )
   &ScaleColor(
    Scale Alpha=&FloatFromCurve(
     FloatCurve=DataInterface() {propertyValues {Curve {keys [
         {arriveTangent=0.7573, interpMode=RCIM_Cubic, time=0, value=0}
         <at 0.07 1 RCIM_Cubic>
         {interpMode=RCIM_Cubic, leaveTangent=-2.7546, time=1, value=0}
        ]}}}
     Scale Curve=&FloatFromCurve(
      FloatCurve=DataInterface() {propertyValues {Curve {keys [
          <at 0 1>
          {arriveTangent=-0.5, interpMode=RCIM_Cubic, leaveTangent=-0.5, time=0.25, value=1}
          {interpMode=RCIM_Cubic, leaveTangent=-0.8975, time=1, value=0.5}
         ]}}}
      CurveIndex=$Emitter.Scalability.DistanceFraction
     )
    )
    Scale RGB=&MakeVectorFromLinearColorRGB(Color=&ColorFromCurve(
      ColorCurve=DataInterface() {propertyValues {
        AlphaCurve {keys [<at 0 1 RCIM_Cubic>]}
        BlueCurve {keys [<at 0.061 0.8594 RCIM_Cubic>,<at 0.25 0.5025>,<at 0.5 0.6927>,<at 0.75 0.354>,<at 0.8984 0.8594 RCIM_Cubic>]}
        GreenCurve {keys [<at 0.061 0.4819 RCIM_Cubic>,<at 0.25 0.6648>,<at 0.5 0.7951>,<at 0.75 0.474>,<at 0.8984 0.4819 RCIM_Cubic>]}
        RedCurve {keys [<at 0.061 0.3939 RCIM_Cubic>,<at 0.25 1>,<at 0.5 1>,<at 0.75 0.7812>,<at 0.8984 0.3939 RCIM_Cubic>]}
       }}
      CurveIndex=Linked() {linkedVariable {name=NPC.FortniteNPC.TimeOfDay_0-1, type=NiagaraFloat}}
     ))
    ScaleA=true
    ScaleRGB=true
   )
  ]
  rendererClasses [&NiagaraSpriteRendererProperties]
  renderers [{rendererClass=&NiagaraSpriteRendererProperties}]
 }
 SphereShockwave {
  emitterSpawnScript [&InitializeAudioPlayer(Audio Player=DataInterface() {propertyValues {
      Attenuation=null
      Concurrency=null
      MaxPlaysPerTick=10
      SoundToPlay=/ToyFoxGameplay/Assets/FX/TEMP/Audio/TEMP_ToyFox_Milkshake_Burst_Cue
      bAllowLoopingOneShotSounds=false
      bLimitPlaysPerTick=true
      bOnlyActiveDuringGameplay=false
      bStopWhenComponentIsDestroyed=true
      ConfigurationUserParameter {parameter {name=None, typeDefHandle {registeredTypeIndex=1}, varData []}}
      ParameterNames []
     }})]
  emitterUpdateScript [
   &EmitterState()
   &SpawnBurst_Instantaneous(Use Spawn Probability=false, UseLoopCountLimit=false)
  ]
  particleSpawnScript [
   &InitializeParticle(
    Color=<rgba 5000 1145.8334 1145.8334 1.2>
    Lifetime=1.35
    Mesh Scale Mode=::NewEnumerator3
    Mesh Uniform Scale=45
    UsePositionOffset=false
    Write Material Random=true
   )
   &CameraOffset(Camera Offset Amount=-500)
   /Niagara/Modules/Audio/PlayAudio()
  ]
  particleUpdateScript [
   &ParticleState()
   &DynamicMaterialParameters(
    Index 0 Param 1=&FloatFromCurve(FloatCurve=DataInterface() {propertyValues {Curve {keys [
         {arriveTangent=2.0031, interpMode=RCIM_Cubic, leaveTangent=2.0031, tangentMode=RCTM_User, time=0, value=0}
         {arriveTangent=1.9873, interpMode=RCIM_Cubic, leaveTangent=1.9873, tangentMode=RCTM_User, time=0.6336, value=1.3072}
         <at 1.0021 1.9784>
        ]}}})
    Index 0 Param 2=&FloatFromCurve(FloatCurve=DataInterface() {propertyValues {Curve {keys [<at 0 15>,<at 0.1612 13>,<at 1 5>]}}})
    Index 0 Param 3=-0.25
    Module.Index0Param0Write=true
    Module.Index0Param1Write=true
    Module.Index0Param2Write=true
    Module.Index0Param3Write=true
   )
   &ScaleMeshSize(Scale Factor=&VectorFromFloat(Value=&FloatFromCurve(FloatCurve=DataInterface() {propertyValues {Curve {keys [{arriveTangent=4.3343, interpMode=RCIM_Cubic, leaveTangent=4.3343, tangentMode=RCTM_User, time=0, value=0.2}, <at 1 2 RCIM_Cubic>]}}})))
   &ScaleColor(
    Scale Alpha=&FloatFromCurve(FloatCurve=DataInterface() {propertyValues {Curve {keys [
         {interpMode=RCIM_Cubic, tangentMode=RCTM_User, time=-0.0034, value=0.1203}
         {arriveTangent=-0.1606, interpMode=RCIM_Cubic, leaveTangent=-0.1606, tangentMode=RCTM_User, time=0.3345, value=0.0495}
         {interpMode=RCIM_Cubic, tangentMode=RCTM_User, time=0.9972, value=-0.0001}
        ]}}})
    Scale RGB=&MakeVectorFromLinearColorRGB(Color=&ColorFromCurve(
      ColorCurve=DataInterface() {propertyValues {
        AlphaCurve {keys [<at 0 1 RCIM_Cubic>]}
        BlueCurve {keys [<at 0.061 0.8594 RCIM_Cubic>,<at 0.25 0.5025>,<at 0.5 0.6927>,<at 0.75 0.354>,<at 0.8984 0.8594 RCIM_Cubic>]}
        GreenCurve {keys [<at 0.061 0.4819 RCIM_Cubic>,<at 0.25 0.6648>,<at 0.5 0.7951>,<at 0.75 0.474>,<at 0.8984 0.4819 RCIM_Cubic>]}
        RedCurve {keys [<at 0.061 0.3939 RCIM_Cubic>,<at 0.25 1>,<at 0.5 1>,<at 0.75 0.7812>,<at 0.8984 0.3939 RCIM_Cubic>]}
       }}
      CurveIndex=Linked() {linkedVariable {name=NPC.FortniteNPC.TimeOfDay_0-1, type=NiagaraFloat}}
     ))
    ScaleA=true
    ScaleRGB=true
   )
  ]
  rendererClasses [&NiagaraMeshRendererProperties]
  renderers [{rendererClass=&NiagaraMeshRendererProperties, rendererIndex=0}]
 }
 Swirl_StreakSpikes_Debris001 {
  emitterUpdateScript [
   &EmitterState(
    Enable Distance Culling=true
    MaxDistance=15000
    Module.bUseMaxDistance=true
    Module.bUseMinDistance=false
    Scalability Mode=::NewEnumerator1
   )
   &SpawnBurst_Instantaneous(
    Spawn Count=&MakeCustomIntFromBool(Bool=$User.FirstBurst, True INT=20, False INT=0)
    Use Spawn Probability=false
    UseLoopCountLimit=false
   )
  ]
  particleSpawnScript [
   &FN_SetSpawnGroup(Number of Spawn Groups=5)
   &InitializeParticle(
    Color=&RandomRangeLinearColor(Minimum=<rgba 0.9427 0.5028 0.2828 1.5>, Maximum=<rgba 1 0.5333 0.3 2>)
    Lifetime=&RandomRangeFloat(Minimum=0.8, Maximum=1.25)
    Mesh Scale Mode=::NewEnumerator3
    Mesh Uniform Scale=20
    Position=&SelectVectorFromArray(Direct Array Index=&SpawnGroup(), Vector Selection Array=$User.Crack Positions)
    Position Mode=::NewEnumerator1
    Sprite Size=&UniformRangedVector2D(Minimum=<67.5 300>, Maximum=<135 450>)
    Sprite Size Mode=::NewEnumerator4
    UsePositionOffset=false
    Write Material Random=true
   )
   &SubUVAnimation(End Frame=3, SubUV Animation Mode:ENiagara_SubUVLookupMode=::NewEnumerator2)
   &self:Quat_Array(Array Index:NiagaraInt32=$Particles.SpawnGroup, Quaternion Array:NiagaraDataInterfaceArrayQuat=$User.Crack Orientations) {module=Quat_Array}
   &CameraOffset(Camera Offset Amount=-5)
   &ShapeLocation(
    Cylinder Height=&SelectFloatFromArray(
     Direct Array Index=&SpawnGroup()
     Float Selection Array=DataInterface() {propertyValues {FloatData [2750,3450,3400,3600,2250]}}
    )
    Cylinder Height Midpoint=0
    Cylinder Radius=250
    Hemicircle Y=true
    MaskSpawnGroup=false
    Offset=&RotateVectorByQuaternion(Vector To Rotate=<0 0 0>, Quaternion=$Particles.Quat)
    Offset Mode=::NewEnumerator0
    Override Seed=false
    Rotation Axis=&RotateVectorByQuaternion(Vector To Rotate=<1 0 0>, Quaternion=$Particles.Quat)
    Rotation Mode=::NewEnumerator1
    Shape Primitive=::NewEnumerator1
    Surface Only=true
    Use Endcaps In Surface Only Mode=false
    UseOwnerScale=true
   )
   &AddVelocity(Velocity=&Add_Vector(A=$Particles.ShapeLocation.ShapeNormal, B=<0 0 2>))
  ]
  particleUpdateScript [
   &ParticleState()
   &Drag(Drag=0.3, Use Linear Drag=true, Use Rotational Drag=true)
   &SolveForcesAndVelocity(Clamp Velocity=false, Limit Acceleration=false)
   &DynamicMaterialParameters(
    Index 0 Param 1=250
    Index 0 Param 4=&FloatFromCurve(FloatCurve=DataInterface() {propertyValues {Curve {keys [{interpMode=RCIM_Cubic, tangentMode=RCTM_User, time=0, value=0}, {interpMode=RCIM_Cubic, tangentMode=RCTM_User, time=1, value=1}]}}})
    Index 1 Param 1=&FloatFromCurve(
     FloatCurve=DataInterface() {propertyValues {Curve {keys [
         {interpMode=RCIM_Cubic, tangentMode=RCTM_User, time=0, value=0}
         {arriveTangent=0.2407, interpMode=RCIM_Cubic, leaveTangent=0.2407, tangentMode=RCTM_User, time=1, value=1}
        ]}}}
     Scale Curve=&RandomRangeFloat(Minimum=0.1, Maximum=1)
    )
    Index 1 Param 3=0.5
    Index 1 Param 4=&FloatFromCurve(FloatCurve=DataInterface() {propertyValues {Curve {keys [<at 0 1>,<at 1 0>]}}}, Scale Curve=-0.3)
    Module.Index0Param0Write=true
    Module.Index0Param1Write=false
    Module.Index0Param2Write=false
    Module.Index0Param3Write=true
    Module.Index1Param0Write=true
    Module.Index1Param1Write=true
    Module.Index1Param2Write=true
    Module.Index1Param3Write=true
    Write Parameter Index 1=true
   )
   &ScaleSpriteSize(Scale Factor=&Vector2DFromCurve(
     Vector2Curve=DataInterface() {propertyValues {
       XCurve {keys [
         {arriveTangent=0.4813, interpMode=RCIM_Cubic, leaveTangent=0.4813, tangentMode=RCTM_User, time=0, value=0.625}
         {arriveTangent=1.25, interpMode=RCIM_Cubic, leaveTangent=1.25, time=0.4, value=0.9716}
         {arriveTangent=2.6138, interpMode=RCIM_Cubic, leaveTangent=2.6138, tangentMode=RCTM_User, time=0.9, value=1.75}
        ]}
       YCurve {keys [
         {arriveTangent=9.4573, interpMode=RCIM_Cubic, leaveTangent=9.4573, tangentMode=RCTM_User, time=0, value=0}
         {arriveTangent=0.8193, interpMode=RCIM_Cubic, leaveTangent=0.8193, tangentMode=RCTM_User, time=0.2, value=1.5019}
         <at 1 1.7992 RCIM_Cubic>
        ]}
      }}
     ScaleCurve=&MakeVector2D(
      X=&FloatFromCurve(
       FloatCurve=DataInterface() {propertyValues {Curve {keys [
           {arriveTangent=-0.2333, interpMode=RCIM_Cubic, leaveTangent=-0.2333, tangentMode=RCTM_User, time=0, value=2}
           {arriveTangent=-3.2303, interpMode=RCIM_Cubic, tangentMode=RCTM_Break, time=1, value=1}
          ]}}}
       Scale Curve=1.75
      )
      Y=&FloatFromCurve(
       FloatCurve=DataInterface() {propertyValues {Curve {keys [
           {arriveTangent=2.5653, interpMode=RCIM_Cubic, leaveTangent=2.5653, tangentMode=RCTM_User, time=0, value=0.75}
           {interpMode=RCIM_Cubic, leaveTangent=5.1925, time=1, value=2.5}
          ]}}}
       Scale Curve=2.5
      )
     )
    ))
   &ScaleColor(
    Scale Alpha=&FloatFromCurve(FloatCurve=DataInterface() {propertyValues {Curve {keys [
         {arriveTangent=9.0725, interpMode=RCIM_Cubic, leaveTangent=9.0725, tangentMode=RCTM_User, time=0, value=0}
         {arriveTangent=-0.0923, interpMode=RCIM_Cubic, leaveTangent=-0.0923, tangentMode=RCTM_User, time=0.075, value=1}
         {arriveTangent=-3.3957, interpMode=RCIM_Cubic, leaveTangent=-3.3957, tangentMode=RCTM_User, time=1, value=0}
        ]}}})
    Scale RGB=&MakeVectorFromLinearColorRGB(Color=&ColorFromCurve(
      ColorCurve=DataInterface() {propertyValues {
        AlphaCurve {keys [<at 0 1 RCIM_Cubic>]}
        BlueCurve {keys [<at 0.061 0.8594 RCIM_Cubic>,<at 0.25 0.1909>,<at 0.5 0.2345>,<at 0.75 0.1909>,<at 0.8984 0.8594 RCIM_Cubic>]}
        GreenCurve {keys [<at 0.061 0.4819 RCIM_Cubic>,<at 0.25 0.2794>,<at 0.5 0.2692>,<at 0.75 0.2794>,<at 0.8984 0.4819 RCIM_Cubic>]}
        RedCurve {keys [<at 0.061 0.3939 RCIM_Cubic>,<at 0.25 0.3594>,<at 0.5 0.3385>,<at 0.75 0.3594>,<at 0.8984 0.3939 RCIM_Cubic>]}
       }}
      CurveIndex=Linked() {linkedVariable {name=NPC.FortniteNPC.TimeOfDay_0-1, type=NiagaraFloat}}
     ))
    ScaleA=true
    ScaleRGB=true
   )
   &FN_Sprite_Attributes(
    Color=&MultiplyLinearColor(
     Linear Color=<rgba 0.5104 0.2143 0.0771 0.95>
     Scale Factor=&MakeFloat4FromVectorAndFloat(
      "Vector (RGB)"=<1 1 1>
      "Float (Alpha)"=&FloatFromCurve(FloatCurve=DataInterface() {propertyValues {Curve {keys [
           {arriveTangent=9.0725, interpMode=RCIM_Cubic, leaveTangent=9.0725, tangentMode=RCTM_User, time=0, value=0}
           {arriveTangent=2, leaveTangent=2, time=0.1, value=0.9847}
           {arriveTangent=-1.0941, interpMode=RCIM_Cubic, leaveTangent=-1.0941, time=0.3, value=1}
           {arriveTangent=-0.0405, interpMode=RCIM_Cubic, leaveTangent=-0.0405, tangentMode=RCTM_User, time=1, value=0}
          ]}}})
     )
    )
    Create Color Attribute=true
    Create SpriteSize Attribute=true
    SpriteSize=&Multiply_Vector2D(
     A=&RandomRangeVector2D(Minimum=<100 120>, Maximum=<112 140>)
     B=&Vector2DFromCurve(
      Vector2Curve=DataInterface() {propertyValues {
        XCurve {keys [
          {arriveTangent=0.4813, interpMode=RCIM_Cubic, leaveTangent=0.4813, tangentMode=RCTM_User, time=0, value=0.625}
          {arriveTangent=1.25, interpMode=RCIM_Cubic, leaveTangent=1.25, time=0.4, value=0.9716}
          {arriveTangent=2.6138, interpMode=RCIM_Cubic, leaveTangent=2.6138, tangentMode=RCTM_User, time=0.9, value=1.75}
         ]}
        YCurve {keys [
          {arriveTangent=9.4573, interpMode=RCIM_Cubic, leaveTangent=9.4573, tangentMode=RCTM_User, time=0, value=0}
          {arriveTangent=0.8193, interpMode=RCIM_Cubic, leaveTangent=0.8193, tangentMode=RCTM_User, time=0.2, value=1.5019}
          <at 1 1.7992 RCIM_Cubic>
         ]}
       }}
      ScaleCurve=&MakeVector2D(
       X=&FloatFromCurve(
        FloatCurve=DataInterface() {propertyValues {Curve {keys [
            {arriveTangent=-0.2333, interpMode=RCIM_Cubic, leaveTangent=-0.2333, tangentMode=RCTM_User, time=0, value=2}
            {arriveTangent=-3.2303, interpMode=RCIM_Cubic, tangentMode=RCTM_Break, time=1, value=1}
           ]}}}
        Scale Curve=0.72
       )
       Y=&FloatFromCurve(
        FloatCurve=DataInterface() {propertyValues {Curve {keys [
            {arriveTangent=2.5653, interpMode=RCIM_Cubic, leaveTangent=2.5653, tangentMode=RCTM_User, time=0, value=0.75}
            {interpMode=RCIM_Cubic, leaveTangent=5.1925, time=1, value=2.5}
           ]}}}
        Scale Curve=1.5
       )
      )
     )
    )
    Write to Module Namespace=true
   )
  ]
  rendererClasses [&NiagaraSpriteRendererProperties]
  renderers [
   {rendererClass=&NiagaraSpriteRendererProperties}
   {rendererClass=&NiagaraSpriteRendererProperties, rendererIndex=1}
  ]
 }
 Swirl_Streaks {
  emitterUpdateScript [
   &EmitterState(Enable Distance Culling=true, MaxDistance=7500, Module.bUseMaxDistance=true, Module.bUseMinDistance=false, Scalability Mode=::NewEnumerator1)
   &SpawnBurst_Instantaneous(Spawn Count=7, Use Spawn Probability=false, UseLoopCountLimit=false)
  ]
  particleSpawnScript [
   &InitializeParticle(
    Color=&RandomRangeLinearColor(Minimum=<rgba 0.9427 0.5028 0.2828 1.5>, Maximum=<rgba 1 0.5333 0.3 2>)
    Lifetime Max=1.7
    Lifetime Mode=::NewEnumerator1
    Lifetime Randomness Mode:ENiagaraRandomnessMode=::NewEnumerator2
    Sprite Size=&UniformRangedVector2D(Minimum=<200 450>, Maximum=<240 480>)
    Sprite Size Mode=::NewEnumerator4
    UsePositionOffset=false
    Write Material Random=true
   )
   &ShapeLocation(
    MaskSpawnGroup=false
    Ring / Disc Distribution Mode:ENiagaraSphereDistributionMode=::NewEnumerator3
    Ring Radius=700
    Rotation Angle=0.25
    Rotation Axis=<0 0 1>
    Rotation Mode=::NewEnumerator1
    Shape Primitive=::NewEnumerator4
    UseOwnerScale=true
   )
   &AddVelocity(Module.Use Falloff Distance=false, Origin Offset=<0 0 -1000>, Override Seed=false, Velocity Mode=::NewEnumerator1, Velocity Speed=50)
   &SubUVAnimation(End Frame=3, SubUV Animation Mode:ENiagara_SubUVLookupMode=::NewEnumerator2)
   &CameraOffset(Camera Offset Amount=75)
  ]
  particleUpdateScript [
   &ParticleState()
   &Drag(Use Linear Drag=true, Use Rotational Drag=true)
   &SolveForcesAndVelocity(Clamp Velocity=false, Limit Acceleration=false)
   &DynamicMaterialParameters(
    Index 0 Param 1=25
    Index 0 Param 4=&FloatFromCurve(FloatCurve=DataInterface() {propertyValues {Curve {keys [{interpMode=RCIM_Cubic, tangentMode=RCTM_User, time=0, value=0}, {interpMode=RCIM_Cubic, tangentMode=RCTM_User, time=1, value=1}]}}})
    Index 1 Param 1=&FloatFromCurve(
     FloatCurve=DataInterface() {propertyValues {Curve {keys [
         {interpMode=RCIM_Cubic, tangentMode=RCTM_User, time=0, value=0}
         {arriveTangent=0.2407, interpMode=RCIM_Cubic, leaveTangent=0.2407, tangentMode=RCTM_User, time=1, value=1}
        ]}}}
     Scale Curve=&RandomRangeFloat(Minimum=0.1, Maximum=1)
    )
    Index 1 Param 3=0.5
    Index 1 Param 4=&FloatFromCurve(FloatCurve=DataInterface() {propertyValues {Curve {keys [<at 0 1>,<at 1 0>]}}}, Scale Curve=-0.3)
    Module.Index0Param0Write=true
    Module.Index0Param1Write=false
    Module.Index0Param2Write=false
    Module.Index0Param3Write=true
    Module.Index1Param0Write=true
    Module.Index1Param1Write=true
    Module.Index1Param2Write=true
    Module.Index1Param3Write=true
    Write Parameter Index 1=true
   )
   &ScaleSpriteSize(Scale Factor=&Vector2DFromCurve(
     Vector2Curve=DataInterface() {propertyValues {
       XCurve {keys [
         {arriveTangent=0.4813, interpMode=RCIM_Cubic, leaveTangent=0.4813, tangentMode=RCTM_User, time=0, value=0.625}
         {arriveTangent=1.25, interpMode=RCIM_Cubic, leaveTangent=1.25, time=0.4, value=0.9716}
         {arriveTangent=2.6138, interpMode=RCIM_Cubic, leaveTangent=2.6138, tangentMode=RCTM_User, time=0.9, value=1.75}
        ]}
       YCurve {keys [
         {arriveTangent=9.4573, interpMode=RCIM_Cubic, leaveTangent=9.4573, tangentMode=RCTM_User, time=0, value=0}
         {arriveTangent=0.8193, interpMode=RCIM_Cubic, leaveTangent=0.8193, tangentMode=RCTM_User, time=0.2, value=1.5019}
         <at 1 1.7992 RCIM_Cubic>
        ]}
      }}
     ScaleCurve=<0.8 1.5>
    ))
   &ScaleSpriteSize(
    Scale Factor=&MakeVector2D(
     X=&FloatFromCurve(
      FloatCurve=DataInterface() {propertyValues {Curve {keys [
          {arriveTangent=-0.2333, interpMode=RCIM_Cubic, leaveTangent=-0.2333, tangentMode=RCTM_User, time=0, value=2}
          {arriveTangent=-3.2303, interpMode=RCIM_Cubic, tangentMode=RCTM_Break, time=1, value=1}
         ]}}}
      Scale Curve=0.9
     )
     Y=&FloatFromCurve(FloatCurve=DataInterface() {propertyValues {Curve {keys [{arriveTangent=1.8705, interpMode=RCIM_Cubic, time=0, value=0.75}, {interpMode=RCIM_Cubic, leaveTangent=5.1925, time=0.3, value=2.5}]}}})
    )
    Scale Sprite Size Mode=::NewEnumerator2
   ) {module=ScaleSpriteSize001}
   &ScaleColor(
    Scale Alpha=&FloatFromCurve(FloatCurve=DataInterface() {propertyValues {Curve {keys [
         {arriveTangent=9.0725, interpMode=RCIM_Cubic, leaveTangent=9.0725, tangentMode=RCTM_User, time=0, value=0}
         {arriveTangent=-0.0923, interpMode=RCIM_Cubic, leaveTangent=-0.0923, tangentMode=RCTM_User, time=0.075, value=1}
         {arriveTangent=-3.3957, interpMode=RCIM_Cubic, leaveTangent=-3.3957, tangentMode=RCTM_User, time=1, value=0}
        ]}}})
    Scale RGB=&MakeVectorFromLinearColorRGB(Color=&ColorFromCurve(
      ColorCurve=DataInterface() {propertyValues {
        AlphaCurve {keys [<at 0 1 RCIM_Cubic>]}
        BlueCurve {keys [<at 0.061 0.8594 RCIM_Cubic>,<at 0.25 0.1909>,<at 0.5 0.2345>,<at 0.75 0.1909>,<at 0.8984 0.8594 RCIM_Cubic>]}
        GreenCurve {keys [<at 0.061 0.4819 RCIM_Cubic>,<at 0.25 0.2794>,<at 0.5 0.2692>,<at 0.75 0.2794>,<at 0.8984 0.4819 RCIM_Cubic>]}
        RedCurve {keys [<at 0.061 0.3939 RCIM_Cubic>,<at 0.25 0.3594>,<at 0.5 0.3385>,<at 0.75 0.3594>,<at 0.8984 0.3939 RCIM_Cubic>]}
       }}
      CurveIndex=Linked() {linkedVariable {name=NPC.FortniteNPC.TimeOfDay_0-1, type=NiagaraFloat}}
     ))
    ScaleA=true
    ScaleRGB=true
   )
  ]
  rendererClasses [&NiagaraSpriteRendererProperties]
  renderers [{rendererClass=&NiagaraSpriteRendererProperties}]
 }
}
summary {
 systemName=NS_TEMP_ToyFox_Milkshake_Burst
 emitters [
  {emitterName=Dirt, rendererClasses [&NiagaraSpriteRendererProperties]}
  {emitterName=Smoke_SphereCenter, rendererClasses [&NiagaraSpriteRendererProperties]}
  {emitterName=Swirl_Streaks, rendererClasses [&NiagaraSpriteRendererProperties]}
  {emitterName=SphereShockwave, rendererClasses [&NiagaraMeshRendererProperties]}
  {emitterName=Sharp_Rocks, rendererClasses [&NiagaraMeshRendererProperties]}
  {emitterName=DustSpikes, rendererClasses [&NiagaraMeshRendererProperties]}
  {
   bEnabled=false
   emitterName=DirectionalBurstPaint
   simTarget=GPUComputeSim
   rendererClasses [&NiagaraSpriteRendererProperties]
  }
  {bEnabled=false, emitterName=Directional_Arm, simTarget=GPUComputeSim, rendererClasses [&NiagaraMeshRendererProperties]}
  {emitterName=Swirl_StreakSpikes_Debris001, rendererClasses [&NiagaraSpriteRendererProperties]}
  {emitterName=Smoke_SphereCracks, rendererClasses [&NiagaraSpriteRendererProperties]}
  {emitterName=MeshChunks, rendererClasses [&NiagaraMeshRendererProperties]}
 ]
 userVariables [
  User.BurstDirection {type=Vector3f, description="", defaultValue {struct=Vector3f, value=<0 0 1>}}
  User.Crack Orientations {
   type=NiagaraDataInterfaceArrayQuat
   description=""
   defaultValue {
    struct=NiagaraExt_VariableValue_DataInterface
    value {
     dataInterface=&self:NiagaraDataInterfaceArrayQuat_0
     dataInterfaceClass=/Script/Niagara.NiagaraDataInterfaceArrayQuat
    }
   }
  }
  User.Crack Positions {
   type=NiagaraDataInterfaceArrayFloat3
   description=""
   defaultValue {
    struct=NiagaraExt_VariableValue_DataInterface
    value {
     dataInterface=&self:NiagaraDataInterfaceArrayFloat3_0
     dataInterfaceClass=&NiagaraDataInterfaceArrayFloat3
    }
   }
  }
  User.FirstBurst {type=NiagaraBool, defaultValue=true}
  User.LifeTimeDuration {type=NiagaraFloat, defaultValue=5, description=""}
 ]
}
