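# Temporary test Niagara effect for the ToyFox Milkshake quake area: dust and debris across an affected area.
# System trace modules pick a location and line-trace it, triggering GPU bursts of streak spikes and debris on hits.
# Smoke_Sphere002 spawns dust over a radius that widens with system age, snapping particle height via landscape sampling.
# Other smoke, chunk and debug emitters are disabled, including the only visible use of QuakeMagnitude.
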NiagaraSystem /ToyFoxGameplay/Assets/FX/TEMP/NS_TEMP_ToyFox_Milkshake_Quake_Area_Test
dependencies {
 usedDataInterfaces [
  &NiagaraDataInterfaceCurve
  &NiagaraDataInterfaceMeshRendererInfo
  &NiagaraDataInterfaceVector2DCurve
  &NiagaraDataInterfaceColorCurve
  &NiagaraDataInterfaceArrayFloat3
  &NiagaraDataInterfaceSpriteRendererInfo
  &NiagaraDataInterfaceCamera
  &NiagaraDataInterfaceCollisionQuery
 ]
 usedDynamicInputs [
  &SelectVectorFromArray
  &Multiply_VectorByFloat
  &UniformRangedVector
  &UniformRangedFloat
  &Add_Vector
  &Subtract_Vector
 ]
 usedModules [
  &SystemState
  &self:SetTraceLocation
  &self:HitscanLineTrace
  &self:SystemSpawnScript.SystemScriptSource.SystemScriptGraph.NiagaraNodeAssignment_1.SetVariables_#1
 ]
}
emitters {
 DistantBigSmoke {
  bEnabled=false
  emitterUpdateScript [
   &EmitterState(
    Enable Distance Culling=true
    MaxDistance=15000
    Min Distance Response=::SleepAndClearParticles
    MinDistance=4500
    Module.bUseMaxDistance=false
    Module.bUseMinDistance=true
    Scalability Mode=::NewEnumerator1
   )
   &SpawnRate(SpawnRate=50, Use Spawn Probability=false)
  ]
  particleSpawnScript [
   &InitializeParticle(
    Color Maximum=<rgba 0.16 0.1292 0.1163 1>
    Color Minimum=<rgba 0.06 0.053 0.0519 0.75>
    Color Mode=::NewEnumerator2
    Lifetime Max=7
    Lifetime Min=4
    Lifetime Mode=::NewEnumerator1
    Sprite Rotation Mode=::NewEnumerator2
    Sprite Size Mode=::NewEnumerator1
    Uniform Sprite Size Max=2100
    Uniform Sprite Size Min=1200
    UsePositionOffset=false
    Write Material Random=true
   ) {module=InitializeParticle001}
   &SphereLocation(
    MaskSpawnGroup=false
    Non Uniform Scale=<1 1 0.1>
    Offset=<0 0 -500>
    Override Seed=false
    Sphere Radius=3000
    Surface Only=true
    UseNonUniformScale=true
    UseOffset=true
    UseSphereOrientation=false
   )
   &AddVelocityInCone(
    Cone Angle=40
    Cone Axis=&Lerp_Vector(A=$System.HitNormal, B=<0 0 1>, Alpha=0.5)
    Module.Use Velocity Falloff On Cone Axis=false
    Velocity Distribution Along Cone Axis=1
    Velocity Strength=&RandomRangeFloat(Minimum=70, Maximum=160)
   )
   &SubUVAnimation(End Frame=44, Start Frame=32, SubUV Animation Mode:ENiagara_SubUVLookupMode=::NewEnumerator2)
  ]
  particleUpdateScript [
   &ParticleState()
   &SpriteRotationRate(Rotation Rate=&Multiply_Float(
     A=&Lerp_Float(A=-100, B=100, Alpha=$Particles.MaterialRandom)
     B=&FloatFromCurve(FloatCurve=DataInterface() {propertyValues {Curve {keys [{arriveTangent=-3.3269, interpMode=RCIM_Cubic, leaveTangent=-3.3269, tangentMode=RCTM_User, time=0, value=1}, <at 0.75 0.2>]}}})
    ))
   &Drag(Drag=0.2, Use Linear Drag=true, Use Rotational Drag=false)
   &WindForce(Wind Speed=<0.5 0 1>, Wind Speed Scale=&RandomRangeFloat(Minimum=100, Maximum=150))
   &SolveForcesAndVelocity(Clamp Velocity=false, Limit Acceleration=false)
   &CameraOffset(Camera Offset Amount=&Multiply_Float(
     A=/Niagara/DynamicInputs/Math/CalculateParticleRadius(
      Radius Calculation Type:ENiagaraCollisionRadiusOptions=::NewEnumerator0
      Particle Radius:NiagaraFloat=1
      Mesh Dimensions:Vector3f=<1 1 1>
      Method for Calculating Particle Radius:ENiagaraCalculateRadiusOptions=::NewEnumerator2
     )
     B=1.75
    ))
   &ScaleSpriteSize(Scale Factor=&Vector2DFromFloat(Value=&FloatFromCurve(
      FloatCurve=DataInterface() {propertyValues {Curve {keys [
          {arriveTangent=0.8984, time=0, value=0.5}
          {arriveTangent=0.1664, interpMode=RCIM_Cubic, leaveTangent=0.1664, tangentMode=RCTM_User, time=1, value=1}
         ]}}}
      Scale Curve=&RemapRange(Input Value To Remap=$Engine.Owner.LODDistance, Input Min=4000, Input Max=10000, Output Min=0.5, Output Max=2)
     )))
   &ScaleColor(
    Scale Alpha=&FloatFromCurve(
     FloatCurve=DataInterface() {propertyValues {Curve {keys [
         {arriveTangent=8.8602, interpMode=RCIM_Cubic, time=0, value=0}
         {arriveTangentWeight=0.1137, interpMode=RCIM_Cubic, leaveTangentWeight=0.2047, tangentMode=RCTM_User, time=0.3, value=1}
         {arriveTangent=-1.9842, interpMode=RCIM_Cubic, leaveTangent=-1.9842, tangentMode=RCTM_User, time=0.75, value=0.75}
         {interpMode=RCIM_Cubic, leaveTangent=-3.4295, time=1, value=0}
        ]}}}
     Scale Curve=&RemapRange(Input Value To Remap=$Engine.Owner.LODDistance, Input Min=4500, Input Max=5200, Output Min=0, Output Max=1)
    )
    ScaleA=true
    ScaleRGB=true
   )
  ]
  rendererClasses [&NiagaraSpriteRendererProperties]
  renderers [{rendererClass=&NiagaraSpriteRendererProperties}]
 }
 GPUChunks {
  bEnabled=false
  simTarget=GPUComputeSim
  emitterUpdateScript [
   &EmitterState(Enable Distance Culling=true, Module.bUseMaxDistance=true, Module.bUseMinDistance=false, Scalability Mode=::NewEnumerator1)
   &SpawnBurst_Instantaneous(
    Spawn Count=&RandomRangeInt(Minimum=7, Maximum=12)
    Spawn Probability=&MakeFloatFromInt(Int=$System.Hit)
    Use Spawn Probability=true
   )
   &SpawnBurst_Instantaneous(
    Spawn Count=&RandomRangeInt(Minimum=6, Maximum=10)
    Spawn Probability=&Multiply_Float(A=$User.QuakeMagnitude, B=&MakeFloatFromInt(Int=$System.Hit))
    Spawn Time=0.08
    Use Spawn Probability=true
   ) {module=SpawnBurst_Instantaneous001}
  ]
  particleSpawnScript [
   &InitializeParticle(
    Color=<rgba 0.4479 0.4159 0.4106 1>
    Lifetime Max=2.25
    Lifetime Min=1.25
    Lifetime Mode=::NewEnumerator1
    Position:Vector3f=$System.SpawnLocation
    Position Mode=::NewEnumerator1
    Sprite Rotation Mode=::NewEnumerator2
    Sprite Size Max=<15 15>
    Sprite Size Min=<8 8>
    Sprite Size Mode=::NewEnumerator2
    UsePositionOffset=false
    Write Material Random=true
   )
   &AddVelocityInCone(
    Cone Angle=70
    Cone Axis=&Lerp_Vector(A=$System.HitNormal, B=<0 0 1>, Alpha=0.5)
    Module.Use Velocity Falloff On Cone Axis=false
    Velocity Distribution Along Cone Axis=2
    Velocity Strength=&RandomRangeFloat(Minimum=600, Maximum=1400)
   )
   &DynamicMaterialParameters(
    Index 0 Param 4=&MaskFloatBySpawnGroup(Masked Float=0, Passthrough Float=1, Spawn Group Mask=&SpawnGroup())
    Module.Param0WriteEnabled=true
    Module.Param1WriteEnabled=false
    Module.Param2WriteEnabled=false
    Module.Param3WriteEnabled=true
   )
   /Niagara/Modules/Update/SubUV/SubUVAnimation(SubUV Animation Mode=::NewEnumerator1)
  ]
  particleUpdateScript [
   &ParticleState()
   &GravityForce(Gravity=<0 0 -1600>)
   &Drag(Drag=0.3, Use Linear Drag=true, Use Rotational Drag=false)
   &DoOnce(Trigger Condition:NiagaraBool=&SetBoolByFloatComparison(A=&VectorLength(Vector=$Particles.Velocity), B=150, Comparison Type=::NewEnumerator6))
   &Collision(
    Advanced Aging Rate=0.5
    Age Colliding Particles=true
    Collision Enabled=&Boolean_AndOperation(Bool1=&SetBoolByFloatComparison(A=$Particles.Age, B=0.3), Bool2=$Particles.DoOnce.TriggeredOnce)
    Friction:NiagaraFloat=&RandomRangeFloat(Minimum=0.5, Maximum=0.7)
    Friction During a Bounce=0.5
    Kill Occluded Particles=false
    Kill Offscreen Particles=true
    Module.Control Rest State Via Surface Normal=false
    Module.EnableMaxCollisionCount=false
    Module.Randomize Collision Normal Vector=true
    Randomize Collision Normal=0.3
    Restitution=&RemapRange(
     Input Value To Remap=&VectorLength(Vector=$Particles.Velocity)
     Input Min=50
     Input Max=600
     Output Min=2.2
     Output Max=0.3
    )
   )
   &SolveForcesAndVelocity(Clamp Velocity=false, Limit Acceleration=false)
   &ScaleSpriteSize(Scale Factor=&Vector2DFromFloat(Value=&FloatFromCurve(
      FloatCurve=DataInterface() {propertyValues {Curve {keys [<at 0.8 1 RCIM_Cubic>, {arriveTangent=-8.4786, interpMode=RCIM_Cubic, leaveTangent=-8.4786, tangentMode=RCTM_User, time=1, value=0}]}}}
      Scale Curve=2
     )))
   &SpriteRotationRate(Rotation Rate=&Multiply_Float(
     A=&FloatFromCurve(
      FloatCurve=DataInterface() {propertyValues {Curve {keys [
          {interpMode=RCIM_Cubic, tangentMode=RCTM_User, time=0, value=-1}
          {arriveTangent=6.8394, interpMode=RCIM_Cubic, leaveTangent=6.8394, tangentMode=RCTM_User, time=0.5, value=0}
          {interpMode=RCIM_Cubic, tangentMode=RCTM_User, time=1, value=1}
         ]}}}
      CurveIndex=&RandomRangeFloat(Minimum=0, Maximum=1)
      Scale Curve=100
     )
     B=&ClampFloat(Float=&MakeFloatFromVector(Vector=$Particles.Velocity, Channel=::NewEnumerator2), Min=-500, Max=500)
    ))
  ]
  rendererClasses [&NiagaraSpriteRendererProperties]
  renderers [{rendererClass=&NiagaraSpriteRendererProperties}]
 }
 RadiusTest {
  bEnabled=false
  simTarget=GPUComputeSim
  emitterUpdateScript [
   &EmitterState()
   &SpawnBurst_Instantaneous(Spawn Time=0.05, Use Spawn Probability=false, UseLoopCountLimit=false)
  ]
  particleSpawnScript [&InitializeParticle(
    Color Mode=::NewEnumerator0
    Lifetime=0.6
    Mesh Scale Mode=::NewEnumerator3
    Position Mode=::NewEnumerator1
    Position Offset=$System.TraceCenterPoint
    Sprite Size Mode=::NewEnumerator3
    Uniform Sprite Size=200
    UsePositionOffset=true
    Write Material Random=true
   )]
  particleUpdateScript [
   &ParticleState()
   /SetVariables(Particles.Position:Vector3f=$System.TraceCenterPoint) {isSetParameters=true}
   &ScaleMeshSize(Scale Factor=&MakeVector(X=$System.TraceRadius, Y=$System.TraceRadius, Z=500))
  ]
  rendererClasses [&NiagaraSpriteRendererProperties, &NiagaraMeshRendererProperties]
  renderers [
   {rendererClass=&NiagaraSpriteRendererProperties}
   {rendererClass=&NiagaraMeshRendererProperties, rendererIndex=1}
  ]
 }
 Smoke {
  bEnabled=false
  simTarget=GPUComputeSim
  emitterUpdateScript [
   &EmitterState(
    Enable Distance Culling=true
    Max Distance Response=::SleepAndClearParticles
    MaxDistance=7000
    Module.bUseMaxDistance=true
    Module.bUseMinDistance=false
    Scalability Mode=::NewEnumerator1
   )
   &SpawnBurst_Instantaneous(
    Spawn Count=&RandomRangeInt(Minimum=3, Maximum=5)
    Spawn Probability=&MakeFloatFromInt(Int=$System.Hit)
    Use Spawn Probability=true
    UseLoopCountLimit=false
   )
  ]
  particleSpawnScript [
   &InitializeParticle(
    Color Maximum=<rgba 0.2135 0.1239 0.0756 1>
    Color Minimum=<rgba 0.2344 0.109 0.0708 0.9>
    Color Mode=::NewEnumerator2
    Lifetime Min=1.5
    Lifetime Mode=::NewEnumerator1
    Position:Vector3f=$System.SpawnLocation
    Position Mode=::NewEnumerator1
    Sprite Rotation Mode=::NewEnumerator2
    Sprite Size Mode=::NewEnumerator1
    Uniform Sprite Size Max=800
    Uniform Sprite Size Min=600
    UsePositionOffset=false
    Write Material Random=true
   ) {module=InitializeParticle001}
   &SphereLocation(
    MaskSpawnGroup=false
    Non Uniform Scale=<1 1 0.25>
    Offset=<0 0 -200>
    Override Seed=false
    Sphere Radius=20
    Surface Only=true
    UseNonUniformScale=true
    UseOffset=true
    UseSphereOrientation=false
   )
   &AddVelocityInCone(
    Cone Angle=70
    Cone Axis=&Lerp_Vector(A=$System.HitNormal, B=<0 0 1>, Alpha=0.5)
    Module.Use Velocity Falloff On Cone Axis=false
    Velocity Distribution Along Cone Axis=1
    Velocity Strength=&RandomRangeFloat(Minimum=1425, Maximum=3600)
   )
  ]
  particleUpdateScript [
   &ParticleState()
   &Drag(Drag=&RandomRangeFloat(Minimum=1, Maximum=1.5), Use Linear Drag=true, Use Rotational Drag=true)
   &SpriteRotationRate(Rotation Rate=&Multiply_Float(
     A=&Lerp_Float(A=-200, B=200, Alpha=$Particles.MaterialRandom)
     B=&FloatFromCurve(FloatCurve=DataInterface() {propertyValues {Curve {keys [
          {arriveTangent=-4.4572, interpMode=RCIM_Cubic, leaveTangent=-4.4572, tangentMode=RCTM_User, time=0, value=1}
          {arriveTangent=-1.0628, interpMode=RCIM_Cubic, leaveTangent=-1.0628, tangentMode=RCTM_User, time=0.2, value=0.4}
          <at 1.0011 0.01>
         ]}}})
    ))
   &WindForce(Wind Speed=<0 0 1>, Wind Speed Scale=250)
   &SolveForcesAndVelocity(Clamp Velocity=false, Limit Acceleration=false)
   &CameraOffset(Camera Offset Amount=&CalculateParticleRadius(
     Radius Calculation Type:ENiagaraCollisionRadiusOptions=::NewEnumerator0
     Particle Radius:NiagaraFloat=1
     Mesh Dimensions:Vector3f=<1 1 1>
     Method for Calculating Particle Radius:ENiagaraCalculateRadiusOptions=::NewEnumerator2
    ))
   &ScaleSpriteSize(Scale Factor=&Vector2DFromFloat(Value=&FloatFromCurve(
      FloatCurve=DataInterface() {propertyValues {Curve {keys [
          {arriveTangent=2.6058, interpMode=RCIM_Cubic, leaveTangent=2.6058, tangentMode=RCTM_User, time=0, value=0.1}
          {arriveTangent=0.7744, interpMode=RCIM_Cubic, leaveTangent=0.7744, tangentMode=RCTM_User, time=0.3, value=0.6}
          {arriveTangent=0.4778, interpMode=RCIM_Cubic, leaveTangent=0.4778, tangentMode=RCTM_User, time=1, value=1}
         ]}}}
      Scale Curve=2
     )))
   &ScaleColor(
    Scale Alpha=&FloatFromCurve(
     FloatCurve=DataInterface() {propertyValues {Curve {keys [
         {arriveTangent=8.8602, interpMode=RCIM_Cubic, leaveTangent=8.8602, tangentMode=RCTM_User, time=0, value=0}
         <at 0.25 1 RCIM_Cubic>
         {interpMode=RCIM_Cubic, leaveTangent=-3.4295, time=1, value=0}
        ]}}}
     Scale Curve=&RemapRange(Input Value To Remap=$Engine.Owner.LODDistance, Input Min=8000, Input Max=14000, Output Min=1, Output Max=0)
    )
    ScaleA=true
    ScaleRGB=true
   )
   /SetVariables(Particles.SubImageIndex=&Lerp_Float(A=20, B=63, Alpha=$Particles.NormalizedAge)) {isSetParameters=true}
  ]
  rendererClasses [&NiagaraSpriteRendererProperties]
  renderers [{rendererClass=&NiagaraSpriteRendererProperties}]
 }
 Smoke_Sphere {
  bEnabled=false
  emitterUpdateScript [
   &EmitterState()
   &SpawnRate(SpawnRate=20, Use Spawn Probability=false)
  ]
  particleSpawnScript [
   &InitializeParticle(
    Color Maximum=<rgba 0.375 0.2328 0.1562 0.45>
    Color Minimum=<rgba 0.375 0.2073 0.1562 0.4>
    Color Mode=::NewEnumerator2
    Lifetime=5
    Sprite Rotation Angle=&RandomRangeFloat(Minimum=-25, Maximum=25)
    Sprite Rotation Mode=::NewEnumerator0
    Sprite Size Mode=::NewEnumerator1
    Uniform Sprite Size Max=2475
    Uniform Sprite Size Min=2402.5
    UsePositionOffset=false
    Write Material Random=true
   )
   &self:LandscapeSample_01() {enabled=false, module=LandscapeSample_01}
   /SetVariables(Particles.Position=&AddVectorToPosition(Position=$Particles.Position, Vector=&Multiply_Vector(A=$System.SpawnLocation, B=<1 1 1>))) {enabled=false, isSetParameters=true, module=SetVariables_#2}
   &ShapeLocation(
    Disc Coverage=1
    MaskSpawnGroup=false
    Override Seed=false
    Ring Radius=&Lerp_Float(
     A=750
     B=4500
     Alpha=&FloatFromCurve(
      FloatCurve=DataInterface() {propertyValues {Curve {keys [
          {arriveTangent=2.2037, interpMode=RCIM_Cubic, leaveTangent=2.2037, tangentMode=RCTM_User, time=0, value=0}
          {interpMode=RCIM_Cubic, tangentMode=RCTM_User, time=1, value=1}
         ]}}}
      CurveIndex=$Emitter.NormalizedLoopAge
     )
    )
    Shape Origin=$Engine.Emitter.SimulationPosition
    Shape Primitive=::NewEnumerator4
    UseOwnerScale=true
   ) {module=ShapeLocation001}
   &ShapeLocation(
    Cylinder Height=10
    Cylinder Height Midpoint=0
    Cylinder Radius=&Lerp_Float(
     A=750
     B=5400
     Alpha=&FloatFromCurve(
      FloatCurve=DataInterface() {propertyValues {Curve {keys [
          {arriveTangent=2.2037, interpMode=RCIM_Cubic, leaveTangent=2.2037, tangentMode=RCTM_User, time=0, value=0}
          {interpMode=RCIM_Cubic, tangentMode=RCTM_User, time=1, value=1}
         ]}}}
      CurveIndex=$Emitter.NormalizedLoopAge
     )
    )
    MaskSpawnGroup=false
    Override Seed=false
    Shape Primitive=::NewEnumerator1
    Surface Only=false
    UseOwnerScale=true
   ) {enabled=false}
   &AddVelocity(Velocity=&Multiply_Vector(A=$Particles.ShapeLocation.ShapeNormal, B=<1 1 0>), Velocity Speed Scale=700) {enabled=false}
   &CameraOffset(Camera Offset Amount=400)
  ]
  particleUpdateScript [
   &ParticleState()
   &ScaleColor(
    Scale Alpha=&FloatFromCurve(FloatCurve=DataInterface() {propertyValues {Curve {keys [
         {arriveTangent=1.4251, interpMode=RCIM_Cubic, time=0, value=0}
         {arriveTangent=9.9593, interpMode=RCIM_Cubic, leaveTangent=0.0706, tangentMode=RCTM_Break, time=0.15, value=1}
         {arriveTangent=-4.3325, interpMode=RCIM_Cubic, leaveTangent=-4.3325, tangentMode=RCTM_User, time=1, value=0}
        ]}}})
    ScaleA=true
    ScaleRGB=true
   )
   /Niagara/Modules/Update/Forces/CurlNoiseForce(Noise Strength=200, Pan Noise=true, Pan Noise Field=<0 0 1>)
   &AccelerationForce(Acceleration=<0 0 -100>)
   &Drag(Drag=0.5, Use Linear Drag=true, Use Rotational Drag=true)
   &SolveForcesAndVelocity(Clamp Velocity=false, Limit Acceleration=false)
   &ScaleSpriteSize(Uniform Curve Sprite Scale=DataInterface() {propertyValues {Curve {keys [
        {arriveTangent=2.4132, interpMode=RCIM_Cubic, time=0, value=1}
        {arriveTangent=2.1876, interpMode=RCIM_Cubic, leaveTangent=2.1876, tangentMode=RCTM_User, time=1, value=2.5}
       ]}}})
   &DynamicMaterialParameters(
    Index 0 Param 1=&FloatFromCurve(FloatCurve=DataInterface() {propertyValues {Curve {keys [{interpMode=RCIM_Cubic, tangentMode=RCTM_User, time=0, value=0}, {interpMode=RCIM_Cubic, tangentMode=RCTM_User, time=1, value=1}]}}})
    Index 0 Param 2=&FloatFromCurve(FloatCurve=DataInterface() {propertyValues {Curve {keys [<at 0.25 0.5 RCIM_Cubic>, {interpMode=RCIM_Cubic, leaveTangent=0.6613, time=1, value=1}]}}})
    Index 0 Param 3=250
    Module.Index0Param0Write=true
    Module.Index0Param1Write=true
    Module.Index0Param2Write=true
    Module.Index0Param3Write=true
   )
   /Niagara/Modules/Update/SubUV/V2/SubUVAnimation(
    Frames Per Second:NiagaraInt32=&RandomRangeInt(Minimum=10, Maximum=12)
    Playback Mode:ENiagara_FPSPlayrate=::NewEnumerator1
    Playback Scale Curve=DataInterface() {propertyValues {Curve {keys [
        {interpMode=RCIM_Cubic, tangentMode=RCTM_User, time=0.25, value=1}
        {arriveTangent=-1.2656, interpMode=RCIM_Cubic, leaveTangent=-1.2656, tangentMode=RCTM_User, time=1, value=0}
       ]}}}
    Random Start Frame=true
    "Sprite Renderer "=DataInterface() {propertyValues {SpriteRenderer=&self:Smoke_Sphere_7.NiagaraSpriteRendererProperties_0}}
    Start Frame Offset:NiagaraInt32=&RandomRangeInt(Minimum=0, Maximum=16)
    SubUV Animation Mode:ENiagara_SubUVLookupModeV2=::NewEnumerator3
    UseEndFrame=true
    UsePlayRateScale=false
    UseStartFrame=true
   )
  ]
  rendererClasses [&NiagaraSpriteRendererProperties]
  renderers [{rendererClass=&NiagaraSpriteRendererProperties}]
 }
 Smoke_Sphere001 {
  bEnabled=false
  emitterUpdateScript [
   &EmitterState()
   &SpawnRate(SpawnRate=20, Use Spawn Probability=false)
  ]
  particleSpawnScript [
   &InitializeParticle(
    Color Maximum=<rgba 0.375 0.2328 0.1562 0.45>
    Color Minimum=<rgba 0.375 0.2073 0.1562 0.4>
    Color Mode=::NewEnumerator2
    Lifetime=5
    Sprite Rotation Angle=&RandomRangeFloat(Minimum=-25, Maximum=25)
    Sprite Rotation Mode=::NewEnumerator0
    Sprite Size Mode=::NewEnumerator1
    Uniform Sprite Size Max=2475
    Uniform Sprite Size Min=2402.5
    UsePositionOffset=false
    Write Material Random=true
   )
   &self:LandscapeSample_01() {enabled=false, module=LandscapeSample_01}
   /SetVariables(Particles.Position=&AddVectorToPosition(Position=$Particles.Position, Vector=&Multiply_Vector(A=$System.SpawnLocation, B=<1 1 1>))) {enabled=false, isSetParameters=true, module=SetVariables_#2}
   &ShapeLocation(
    Disc Coverage=1
    MaskSpawnGroup=false
    Override Seed=false
    Ring Radius=&Lerp_Float(
     A=750
     B=4500
     Alpha=&FloatFromCurve(
      FloatCurve=DataInterface() {propertyValues {Curve {keys [
          {arriveTangent=2.2037, interpMode=RCIM_Cubic, leaveTangent=2.2037, tangentMode=RCTM_User, time=0, value=0}
          {interpMode=RCIM_Cubic, tangentMode=RCTM_User, time=1, value=1}
         ]}}}
      CurveIndex=$Emitter.NormalizedLoopAge
     )
    )
    Shape Origin=$Engine.Emitter.SimulationPosition
    Shape Primitive=::NewEnumerator4
    UseOwnerScale=true
   ) {module=ShapeLocation001}
   &ShapeLocation(
    Cylinder Height=10
    Cylinder Height Midpoint=0
    Cylinder Radius=&Lerp_Float(
     A=750
     B=5400
     Alpha=&FloatFromCurve(
      FloatCurve=DataInterface() {propertyValues {Curve {keys [
          {arriveTangent=2.2037, interpMode=RCIM_Cubic, leaveTangent=2.2037, tangentMode=RCTM_User, time=0, value=0}
          {interpMode=RCIM_Cubic, tangentMode=RCTM_User, time=1, value=1}
         ]}}}
      CurveIndex=$Emitter.NormalizedLoopAge
     )
    )
    MaskSpawnGroup=false
    Override Seed=false
    Shape Primitive=::NewEnumerator1
    Surface Only=false
    UseOwnerScale=true
   ) {enabled=false}
   &AddVelocity(Velocity=&Multiply_Vector(A=$Particles.ShapeLocation.ShapeNormal, B=<1 1 0>), Velocity Speed Scale=700) {enabled=false}
   &CameraOffset(Camera Offset Amount=400)
  ]
  particleUpdateScript [
   &ParticleState()
   &ScaleColor(
    Scale Alpha=&FloatFromCurve(FloatCurve=DataInterface() {propertyValues {Curve {keys [
         {arriveTangent=1.4251, interpMode=RCIM_Cubic, time=0, value=0}
         {arriveTangent=9.9593, interpMode=RCIM_Cubic, leaveTangent=0.0706, tangentMode=RCTM_Break, time=0.15, value=1}
         {arriveTangent=-4.3325, interpMode=RCIM_Cubic, leaveTangent=-4.3325, tangentMode=RCTM_User, time=1, value=0}
        ]}}})
    ScaleA=true
    ScaleRGB=true
   )
   &CurlNoiseForce(Noise Strength=200, Pan Noise=true, Pan Noise Field=<0 0 1>)
   &AccelerationForce(Acceleration=<0 0 -100>)
   &Drag(Drag=0.5, Use Linear Drag=true, Use Rotational Drag=true)
   &SolveForcesAndVelocity(Clamp Velocity=false, Limit Acceleration=false)
   &ScaleSpriteSize(Uniform Curve Sprite Scale=DataInterface() {propertyValues {Curve {keys [
        {arriveTangent=2.4132, interpMode=RCIM_Cubic, time=0, value=1}
        {arriveTangent=2.1876, interpMode=RCIM_Cubic, leaveTangent=2.1876, tangentMode=RCTM_User, time=1, value=2.5}
       ]}}})
   &DynamicMaterialParameters(Module.Index0Param0Write=true, Module.Index0Param1Write=true, Module.Index0Param2Write=true, Module.Index0Param3Write=true)
   &SubUVAnimation(
    Frames Per Second:NiagaraInt32=&RandomRangeInt(Minimum=10, Maximum=12)
    Playback Mode:ENiagara_FPSPlayrate=::NewEnumerator1
    Playback Scale Curve=DataInterface() {propertyValues {Curve {keys [
        {interpMode=RCIM_Cubic, tangentMode=RCTM_User, time=0.25, value=1}
        {arriveTangent=-1.2656, interpMode=RCIM_Cubic, leaveTangent=-1.2656, tangentMode=RCTM_User, time=1, value=0}
       ]}}}
    Random Start Frame=true
    "Sprite Renderer "=DataInterface() {propertyValues {SpriteRenderer=&self:Smoke_Sphere001_3.NiagaraSpriteRendererProperties_0}}
    Start Frame Offset:NiagaraInt32=&RandomRangeInt(Minimum=0, Maximum=16)
    SubUV Animation Mode:ENiagara_SubUVLookupModeV2=::NewEnumerator3
    UseEndFrame=true
    UsePlayRateScale=false
    UseStartFrame=true
   )
  ]
  rendererClasses [&NiagaraSpriteRendererProperties]
  renderers [{rendererClass=&NiagaraSpriteRendererProperties}]
 }
 Smoke_Sphere002 {
  simTarget=GPUComputeSim
  emitterUpdateScript [
   &EmitterState(Life Cycle Mode=::NewEnumerator1, Loop Behavior=::NewEnumerator1, Loop Duration=$User.LifeTimeDuration, UseLoopDelay=true)
   &SpawnRate(SpawnRate=14, Use Spawn Probability=false)
  ]
  particleSpawnScript [
   &InitializeParticle(
    Color Maximum=<rgba 0.5 0.2664 0.1406 0.7>
    Color Minimum=<rgba 0.4688 0.2591 0.1953 0.6>
    Color Mode=::NewEnumerator2
    Lifetime=3
    Sprite Rotation Angle=&RandomRangeFloat(Minimum=-25, Maximum=25)
    Sprite Rotation Mode=::NewEnumerator0
    Sprite Size Mode=::NewEnumerator1
    Uniform Sprite Size Max=2475
    Uniform Sprite Size Min=2402.5
    UsePositionOffset=false
    Write Material Random=true
   )
   &ShapeLocation(
    Cylinder Height=1
    Cylinder Height Midpoint=0
    Cylinder Radius=&Lerp_Float(
     A=750
     B=3000
     Alpha=&FloatFromCurve(
      FloatCurve=DataInterface() {propertyValues {Curve {keys [
          {arriveTangent=-0.0443, interpMode=RCIM_Cubic, leaveTangent=-0.0443, tangentMode=RCTM_User, time=0, value=0}
          {arriveTangent=0.383, interpMode=RCIM_Cubic, leaveTangent=0.383, tangentMode=RCTM_User, time=3, value=1}
         ]}}}
      CurveIndex=$System.Age
     )
    )
    MaskSpawnGroup=false
    Override Seed=false
    Shape Primitive=::NewEnumerator1
    Surface Only=false
    UseOwnerScale=true
   )
   &AddVelocity(Velocity=&Multiply_Vector(A=$Particles.ShapeLocation.ShapeNormal, B=<4 1.5 0>), Velocity Speed Scale=500)
   &CameraOffset(Camera Offset Amount=100)
  ]
  particleUpdateScript [
   &ParticleState()
   &self:LandscapeSample_01() {module=LandscapeSample_01}
   &ScaleColor(
    Scale Alpha=&FloatFromCurve(FloatCurve=DataInterface() {propertyValues {Curve {keys [
         {arriveTangent=1.4251, interpMode=RCIM_Cubic, time=0, value=0}
         {arriveTangent=9.9593, leaveTangent=0.0706, time=0.15, value=1}
         {arriveTangent=-1.1765, interpMode=RCIM_Cubic, leaveTangent=-1.1765, time=0.6, value=1}
         {arriveTangent=-6.7209, interpMode=RCIM_Cubic, leaveTangent=-6.7209, tangentMode=RCTM_User, time=1, value=0}
        ]}}})
    ScaleA=true
    ScaleRGB=true
   )
   &CurlNoiseForce(Noise Strength=200, Pan Noise=true, Pan Noise Field=<0 0 1>)
   &Drag(Use Linear Drag=true, Use Rotational Drag=true)
   &SolveForcesAndVelocity(Clamp Velocity=false, Limit Acceleration=false)
   &ScaleSpriteSize(Uniform Curve Sprite Scale=DataInterface() {propertyValues {Curve {keys [
        {arriveTangent=2.3249, interpMode=RCIM_Cubic, leaveTangent=2.3249, tangentMode=RCTM_User, time=0, value=1}
        {arriveTangent=2.1876, interpMode=RCIM_Cubic, leaveTangent=2.1876, tangentMode=RCTM_User, time=1, value=2.5}
       ]}}})
   &DynamicMaterialParameters(
    Index 0 Param 2=&FloatFromCurve(FloatCurve=DataInterface() {propertyValues {Curve {keys [<at 0.25 0.5 RCIM_Cubic>, {interpMode=RCIM_Cubic, leaveTangent=0.6613, time=1, value=1}]}}})
    Index 0 Param 3=300
    Module.Index0Param0Write=true
    Module.Index0Param1Write=true
    Module.Index0Param2Write=true
    Module.Index0Param3Write=true
   )
   &SubUVAnimation(
    Frames Per Second:NiagaraInt32=&RandomRangeInt(Minimum=10, Maximum=12)
    Playback Mode:ENiagara_FPSPlayrate=::NewEnumerator1
    Playback Scale Curve=DataInterface() {propertyValues {Curve {keys [
        {interpMode=RCIM_Cubic, tangentMode=RCTM_User, time=0.25, value=1}
        {arriveTangent=-1.2656, interpMode=RCIM_Cubic, leaveTangent=-1.2656, tangentMode=RCTM_User, time=1, value=0}
       ]}}}
    Random Start Frame=true
    "Sprite Renderer "=DataInterface() {propertyValues {SpriteRenderer=&self:Smoke_Sphere002_5.NiagaraSpriteRendererProperties_0}}
    Start Frame Offset:NiagaraInt32=&RandomRangeInt(Minimum=0, Maximum=16)
    SubUV Animation Mode:ENiagara_SubUVLookupModeV2=::NewEnumerator3
    UseEndFrame=true
    UsePlayRateScale=false
    UseStartFrame=true
   )
   /SetVariables() {enabled=false, isSetParameters=true}
   /SetVariables(Particles.Position=&MakeVector(
     X=&MakeFloatFromVector(Vector=Linked() {linkedVariable {name=Particles.Position, type=NiagaraPosition}}, Channel=::NewEnumerator0)
     Y=&MakeFloatFromVector(Vector=Linked() {linkedVariable {name=Particles.Position, type=NiagaraPosition}}, Channel=::NewEnumerator1)
     Z=$Particles.Height
    )) {isSetParameters=true}
  ]
  rendererClasses [&NiagaraSpriteRendererProperties]
  renderers [{rendererClass=&NiagaraSpriteRendererProperties}]
 }
 Swirl_StreakSpikes_Debris {
  simTarget=GPUComputeSim
  emitterUpdateScript [
   &EmitterState(Enable Distance Culling=true, MaxDistance=7000, Module.bUseMaxDistance=true, Module.bUseMinDistance=false, Scalability Mode=::NewEnumerator1)
   &SpawnBurst_Instantaneous(
    Spawn Probability=&MakeFloatFromInt(Int=$System.Hit)
    Use Spawn Probability=true
    UseLoopCountLimit=false
   )
   &SpawnBurst_Instantaneous(
    Spawn Group=1
    Spawn Probability=&MakeFloatFromInt(Int=$System.Hit)
    Use Spawn Probability=true
    UseLoopCountLimit=false
   ) {module=SpawnBurst_Instantaneous001}
   &SpawnBurst_Instantaneous(
    Spawn Count=&RandomRangeInt(Minimum=5, Maximum=9)
    Spawn Group=2
    Spawn Probability=&MakeFloatFromInt(Int=$System.Hit)
    Use Spawn Probability=true
    UseLoopCountLimit=false
   ) {module=SpawnBurst_Instantaneous002}
  ]
  particleSpawnScript [
   &InitializeParticle(
    Color=&MaskLinearColorBySpawnGroup(
     Masked Linear Color=&RandomRangeLinearColor(Minimum=<rgba 0.5104 0.2325 0.1037 1.5>, Maximum=<rgba 0.3385 0.1506 0.0635 2>)
     Passthrough Linear Color=&MaskLinearColorBySpawnGroup(Masked Linear Color=<rgba 0.5104 0.2143 0.0771 0.95>, Passthrough Linear Color=<rgba 0.4479 0.2279 0.126 1.5>, Spawn Group Mask=1)
    )
    Lifetime=&MaskFloatBySpawnGroup(
     Masked Float=&RandomRangeFloat(Minimum=1.2, Maximum=1.5)
     Passthrough Float=&RandomRangeFloat(Minimum=0.8, Maximum=1)
     Spawn Group Mask=1
    )
    Mesh Scale Mode=::NewEnumerator3
    Mesh Uniform Scale=20
    Position=$System.SpawnLocation
    Position Mode=::NewEnumerator1
    Sprite Size=&MaskVector2dBySpawnGroup(
     Masked Vector2D=&UniformRangedVector2D(Minimum=<180 350>, Maximum=<240 550>)
     Passthrough Vector2D=&MaskVector2dBySpawnGroup(
      Masked Vector2D=<450 450>
      Passthrough Vector2D=&RandomRangeVector2D(Minimum=<8 8>, Maximum=<15 15>)
      Spawn Group Mask=1
     )
     Spawn Group Mask=0
    )
    Sprite Size Mode=::NewEnumerator4
    UsePositionOffset=false
    Write Material Random=true
   )
   &AddVelocity(Velocity=&RandomRangeVector(Minimum=<0 0 0>, Maximum=<0 0 4>)) {enabled=false}
   &AddVelocity(
    Module.Use Falloff Distance=false
    Origin Offset=&RandomRangeVector(Minimum=<0 0 -1500>, Maximum=<0 0 -1000>)
    Override Seed=false
    Velocity Mode=::NewEnumerator1
    Velocity Speed=1
   ) {enabled=false, module=AddVelocity001}
   &AddVelocityInCone(
    Cone Angle=&MaskFloatBySpawnGroup(Masked Float=90, Passthrough Float=0, Spawn Group Mask=2)
    Cone Axis=&Lerp_Vector(A=$System.HitNormal, B=<0 0 1>, Alpha=0.5)
    Module.Use Velocity Falloff On Cone Axis=false
    Velocity Strength=&MaskFloatBySpawnGroup(Masked Float=&RandomRangeFloat(Minimum=850, Maximum=1500), Passthrough Float=1, Spawn Group Mask=2)
   )
   &SubUVAnimation(End Frame=3, SubUV Animation Mode:ENiagara_SubUVLookupMode=::NewEnumerator2)
   &CameraOffset()
  ]
  particleUpdateScript [
   &ParticleState()
   &Drag(Drag=0.3, Use Linear Drag=true, Use Rotational Drag=false)
   &GravityForce(Gravity=&MaskVectorBySpawnGroup(Masked Vector=<0 0 -1250>, Passthrough Vector=<0 0 0>, Spawn Group Mask=2))
   &SolveForcesAndVelocity(Clamp Velocity=false, Limit Acceleration=false)
   &DynamicMaterialParameters(
    Index 0 Param 1=200
    Index 0 Param 4=&FloatFromCurve(FloatCurve=DataInterface() {propertyValues {Curve {keys [{interpMode=RCIM_Cubic, tangentMode=RCTM_User, time=0, value=0}, {interpMode=RCIM_Cubic, tangentMode=RCTM_User, time=1, value=1}]}}})
    Index 1 Param 1=&FloatFromCurve(
     FloatCurve=DataInterface() {propertyValues {Curve {keys [
         {interpMode=RCIM_Cubic, tangentMode=RCTM_User, time=0, value=0}
         {arriveTangent=0.2407, interpMode=RCIM_Cubic, leaveTangent=0.2407, tangentMode=RCTM_User, time=1, value=1}
        ]}}}
     Scale Curve=&RandomRangeFloat(Minimum=0.1, Maximum=1)
    )
    Index 1 Param 3=0.5
    Index 1 Param 4=&FloatFromCurve(FloatCurve=DataInterface() {propertyValues {Curve {keys [<at 0 1>,<at 1 0>]}}}, Scale Curve=-0.3)
    Module.Index0Param0Write=true
    Module.Index0Param1Write=false
    Module.Index0Param2Write=false
    Module.Index0Param3Write=true
    Module.Index1Param0Write=true
    Module.Index1Param1Write=true
    Module.Index1Param2Write=true
    Module.Index1Param3Write=true
    Write Parameter Index 1=true
   )
   &ScaleSpriteSize(Scale Factor=&Vector2DFromCurve(
     Vector2Curve=DataInterface() {propertyValues {
       XCurve {keys [
         {arriveTangent=0.4813, interpMode=RCIM_Cubic, leaveTangent=0.4813, tangentMode=RCTM_User, time=0, value=0.625}
         {arriveTangent=1.25, interpMode=RCIM_Cubic, leaveTangent=1.25, time=0.4, value=0.9716}
         {arriveTangent=2.6138, interpMode=RCIM_Cubic, leaveTangent=2.6138, tangentMode=RCTM_User, time=0.9, value=1.75}
        ]}
       YCurve {keys [
         {arriveTangent=9.4573, interpMode=RCIM_Cubic, leaveTangent=9.4573, tangentMode=RCTM_User, time=0, value=0}
         {arriveTangent=0.8193, interpMode=RCIM_Cubic, leaveTangent=0.8193, tangentMode=RCTM_User, time=0.2, value=1.5019}
         <at 1 1.7992 RCIM_Cubic>
        ]}
      }}
     ScaleCurve=<0.8 1.5>
    ))
   &ScaleSpriteSize(
    Scale Factor=&MakeVector2D(
     X=&FloatFromCurve(
      FloatCurve=DataInterface() {propertyValues {Curve {keys [
          {arriveTangent=-0.2333, interpMode=RCIM_Cubic, leaveTangent=-0.2333, tangentMode=RCTM_User, time=0, value=2}
          {arriveTangent=-3.2303, interpMode=RCIM_Cubic, tangentMode=RCTM_Break, time=1, value=1}
         ]}}}
      Scale Curve=0.9
     )
     Y=&FloatFromCurve(FloatCurve=DataInterface() {propertyValues {Curve {keys [
          {arriveTangent=2.5653, interpMode=RCIM_Cubic, leaveTangent=2.5653, tangentMode=RCTM_User, time=0, value=0.75}
          {interpMode=RCIM_Cubic, leaveTangent=5.1925, time=1, value=2.5}
         ]}}})
    )
    Scale Sprite Size Mode=::NewEnumerator2
   ) {module=ScaleSpriteSize001}
   &ScaleColor(
    Scale Alpha=&MaskFloatBySpawnGroup(
     Masked Float=&FloatFromCurve(FloatCurve=DataInterface() {propertyValues {Curve {keys [
          {arriveTangent=9.0725, interpMode=RCIM_Cubic, leaveTangent=9.0725, tangentMode=RCTM_User, time=0, value=0}
          {arriveTangent=-0.0923, interpMode=RCIM_Cubic, leaveTangent=-0.0923, tangentMode=RCTM_User, time=0.075, value=1}
          {arriveTangent=-1.8476, interpMode=RCIM_Cubic, leaveTangent=-1.8476, tangentMode=RCTM_User, time=1, value=0}
         ]}}})
     Passthrough Float=&MaskFloatBySpawnGroup(
      Masked Float=&FloatFromCurve(FloatCurve=DataInterface() {propertyValues {Curve {keys [
           {arriveTangent=9.0725, interpMode=RCIM_Cubic, leaveTangent=9.0725, tangentMode=RCTM_User, time=0, value=0}
           {arriveTangent=2, leaveTangent=2, time=0.1, value=0.9847}
           {arriveTangent=-1.0941, interpMode=RCIM_Cubic, leaveTangent=-1.0941, time=0.3, value=1}
           {arriveTangent=-0.0405, interpMode=RCIM_Cubic, leaveTangent=-0.0405, tangentMode=RCTM_User, time=1, value=0}
          ]}}})
      Passthrough Float=1
      Spawn Group Mask=1
     )
    )
    ScaleA=true
    ScaleRGB=true
   )
   /SetVariables(Particles.RcokSize:Vector2f=&Multiply_Vector2D(
     A=$Particles.Initial.SpriteSize
     B=&Vector2DFromFloat(Value=&FloatFromCurve(
       FloatCurve=DataInterface() {propertyValues {Curve {keys [<at 0.8 1 RCIM_Cubic>, {arriveTangent=-8.4786, interpMode=RCIM_Cubic, leaveTangent=-8.4786, tangentMode=RCTM_User, time=1, value=0}]}}}
       Scale Curve=&RandomRangeFloat(Minimum=1.75, Maximum=3)
      ))
    )) {isSetParameters=true}
  ]
  rendererClasses [&NiagaraSpriteRendererProperties]
  renderers [
   {rendererClass=&NiagaraSpriteRendererProperties}
   {rendererClass=&NiagaraSpriteRendererProperties, rendererIndex=1}
   {rendererClass=&NiagaraSpriteRendererProperties, rendererIndex=2}
  ]
 }
 TraceDebug {
  bEnabled=false
  emitterUpdateScript [
   &EmitterState()
   &SpawnBurst_Instantaneous(Use Spawn Probability=false, UseLoopCountLimit=false)
  ]
  particleSpawnScript [
   &InitializeParticle(
    Color=<rgba 500 500 500 1>
    Lifetime=$System.CurrentLoopDuration
    Sprite Size Mode=::NewEnumerator3
    Uniform Sprite Size=30
    UsePositionOffset=false
    Write Material Random=true
   )
   &AddVelocity(Velocity=<0 0 -1>)
  ]
  particleUpdateScript [
   &ParticleState()
   /SetVariables(Particles.Position:Vector3f=$System.TraceStart) {isSetParameters=true}
   &ScaleSpriteSize(
    Scale Factor=&MakeVector2D(
     X=0.25
     Y=&Divide_Float(
      A=&Subtract_Float(
       A=&MakeFloatFromVector(Vector=Linked() {linkedVariable {name=Particles.Position, type=NiagaraPosition}}, Channel=::NewEnumerator2)
       B=&MakeFloatFromVector(Vector=$System.HitPos, Channel=::NewEnumerator2)
      )
      B=20
     )
    )
    Scale Sprite Size Mode=::NewEnumerator2
   )
   &SolveForcesAndVelocity(Clamp Velocity=false, Limit Acceleration=false)
  ]
  rendererClasses [&NiagaraSpriteRendererProperties]
  renderers [
   {rendererClass=&NiagaraSpriteRendererProperties}
   {rendererClass=&NiagaraSpriteRendererProperties, rendererIndex=1}
  ]
 }
}
summary {
 systemName=NS_TEMP_ToyFox_Milkshake_Quake_Area_Test
 emitters [
  {bEnabled=false, emitterName=Smoke, simTarget=GPUComputeSim, rendererClasses [&NiagaraSpriteRendererProperties]}
  {bEnabled=false, emitterName=TraceDebug, rendererClasses [&NiagaraSpriteRendererProperties]}
  {bEnabled=false, emitterName=DistantBigSmoke, rendererClasses [&NiagaraSpriteRendererProperties]}
  {
   bEnabled=false
   emitterName=RadiusTest
   simTarget=GPUComputeSim
   rendererClasses [&NiagaraSpriteRendererProperties, &NiagaraMeshRendererProperties]
  }
  {emitterName=Swirl_StreakSpikes_Debris, simTarget=GPUComputeSim, rendererClasses [&NiagaraSpriteRendererProperties]}
  {bEnabled=false, emitterName=Smoke_Sphere, rendererClasses [&NiagaraSpriteRendererProperties]}
  {bEnabled=false, emitterName=GPUChunks, simTarget=GPUComputeSim, rendererClasses [&NiagaraSpriteRendererProperties]}
  {bEnabled=false, emitterName=Smoke_Sphere001, rendererClasses [&NiagaraSpriteRendererProperties]}
  {emitterName=Smoke_Sphere002, simTarget=GPUComputeSim, rendererClasses [&NiagaraSpriteRendererProperties]}
 ]
 userVariables [
  User.LifeTimeDuration {type=NiagaraFloat, defaultValue=5, description=""}
  User.New Landscape Sample_Particles_Position_Z_A {
   type=NiagaraDataInterfaceLandscape
   description=""
   defaultValue {
    struct=NiagaraExt_VariableValue_DataInterface
    value {
     dataInterface=&self:NiagaraDataInterfaceLandscape_0
     dataInterfaceClass=&NiagaraDataInterfaceLandscape
    }
   }
  }
  User.QuakeMagnitude {type=NiagaraFloat, defaultValue=1, description=""}
  User.QuakeRadius {type=NiagaraFloat, defaultValue=6000, description=""}
 ]
}
