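# Temporary Niagara glow effect, named for a ToyFox pedestal, combining sprites and lights with adjustable intensity and light radius.
# One burst emitter spawns persistent particles, using spawn-group masks to vary offset, size and light enablement.
# A system-level normalized intensity drives camera offset, color curves, sprite scale, material parameters and flicker frequency, while User.Intensity scales RGB.
# A phase-offset waveform jitter modulates sprite size, light radius and material parameters.
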NiagaraSystem /ToyFoxGameplay/Assets/FX/TEMP/NS_TEMP_ToyFox_PedestalGlow
dependencies {
 usedDataInterfaces [&NiagaraDataInterfaceCurve, &NiagaraDataInterfaceVectorCurve]
 usedDynamicInputs [&ClampFloat, &RemapRange]
 usedModules [
  &SystemState
  &self:SystemSpawnScript.SystemScriptSource.SystemScriptGraph.NiagaraNodeAssignment_1.SetVariables_#1
 ]
}
emitters {Lights {
  emitterUpdateScript [
   &EmitterState(
    Inactive Response=::NewEnumerator1
    Life Cycle Mode=::NewEnumerator1
    Loop Behavior=::NewEnumerator1
    Loop Duration Mode=::NewEnumerator1
    UseLoopDelay=false
   )
   &SpawnBurst_Instantaneous(Use Spawn Probability=false, UseLoopCountLimit=false)
  ]
  particleSpawnScript [
   &InitializeParticle(
    Color=<rgba 15 3.5326 0.2197 1>
    Position Offset=&MaskVectorBySpawnGroup(Masked Vector=<0 0 0>, Passthrough Vector=<0 0 1800>)
    Position Offset Coordinate Space=::Local
    Sprite Size Mode=::NewEnumerator3
    Uniform Sprite Size=&MaskFloatBySpawnGroup(Masked Float=2500, Passthrough Float=5000)
    UsePositionOffset=true
    Write Material Random=true
   )
   /SetVariables(Particles.LightEnabled:NiagaraBool=/Niagara/DynamicInputs/Helpers/MaskBoolBySpawnGroup(Masked Bool:NiagaraBool=false, Passthrough Bool:NiagaraBool=true, Spawn Group Mask:NiagaraInt32=0)) {isSetParameters=true}
  ]
  particleUpdateScript [
   &ParticleState(Kill Particles When Lifetime Has Elapsed=false)
   &CameraOffset(Camera Offset Amount=&Lerp_Float(
     A=300
     B=2000
     Alpha=&FloatFromCurve(
      FloatCurve=DataInterface() {propertyValues {Curve {keys [
          {arriveTangent=0.6328, interpMode=RCIM_Cubic, leaveTangent=0.6328, tangentMode=RCTM_User, time=0, value=0}
          {arriveTangent=2.2092, interpMode=RCIM_Cubic, leaveTangent=2.2092, tangentMode=RCTM_User, time=0.9, value=0.85}
          {interpMode=RCIM_Cubic, leaveTangent=1.512, time=1, value=1}
         ]}}}
      Scale Curve=$System.NormalizedIntensity
     )
    ))
   &ScaleColor(
    Scale RGB=&VectorFromCurve(
     VectorCurve=DataInterface() {propertyValues {
       XCurve {keys [<at 0 1 RCIM_Cubic>,<at 1 0.3>]}
       YCurve {keys [<at 0.25 1 RCIM_Cubic>,<at 1 0.4 RCIM_Cubic>]}
       ZCurve {keys [<at 0.25 1 RCIM_Cubic>,<at 1 2>]}
      }}
     CurveIndex=$System.NormalizedIntensity
     Scale Curve=&VectorFromFloat(Value=$User.Intensity)
    )
    ScaleA=true
    ScaleRGB=true
   )
   /SetVariables(Particles.Jitter:NiagaraFloat=&Waveform(
     Global Phase=&Add_Float(A=$System.Age, B=&Multiply_FloatByInt(Float=0.375, Integer=&SpawnGroup()))
     "Waveform [1]"=::NewEnumerator12
     "[1] Frequency"=&Lerp_Float(A=10, B=30, Alpha=$System.NormalizedIntensity)
     "[1] Amplitude Min/Max"=<0.9 1>
    )) {isSetParameters=true}
   /SetVariables(Particles.LightRadius=&Multiply_Float(
     A=&Multiply_Float(A=$Particles.Jitter, B=&MakeFloatFromVector2d(Vector 2D=$Particles.SpriteSize))
     B=$User.LightRadiusOverride
    )) {isSetParameters=true}
   /SetVariables(Particles.SpriteColor=&MultiplyLinearColor(Linear Color=$Particles.Color, Scale Factor=&MakeVector4(X=1, Y=1, Z=1, W=0.05))) {isSetParameters=true}
   &DynamicMaterialParameters(
    Index 0 Param 1=&Lerp_Float(A=11, B=5, Alpha=$Particles.Jitter)
    Index 0 Param 2=&Lerp_Float(
     A=4
     B=30
     Alpha=&FloatFromCurve(
      FloatCurve=DataInterface() {propertyValues {Curve {keys [
          {arriveTangent=0.6328, interpMode=RCIM_Cubic, leaveTangent=0.6328, tangentMode=RCTM_User, time=0, value=0}
          {arriveTangent=2.2092, interpMode=RCIM_Cubic, leaveTangent=2.2092, tangentMode=RCTM_User, time=0.9, value=0.85}
          {interpMode=RCIM_Cubic, leaveTangent=1.512, time=1, value=1}
         ]}}}
      Scale Curve=$System.NormalizedIntensity
     )
    )
    Module.Param0WriteEnabled=true
    Module.Param1WriteEnabled=true
    Module.Param2WriteEnabled=true
    Module.Param3WriteEnabled=true
   )
   &ScaleSpriteSize(
    Uniform Curve Index=$System.NormalizedIntensity
    Uniform Curve Scale=$Particles.Jitter
    Uniform Curve Sprite Scale=DataInterface() {propertyValues {Curve {keys [
        {arriveTangent=1.7436, interpMode=RCIM_Cubic, leaveTangent=1.7436, tangentMode=RCTM_User, time=0, value=0.5}
        {arriveTangent=0.7173, interpMode=RCIM_Cubic, leaveTangent=0.7173, tangentMode=RCTM_User, time=0.25, value=1}
        {arriveTangent=5.0539, interpMode=RCIM_Cubic, leaveTangent=5.0539, tangentMode=RCTM_User, time=1, value=1}
       ]}}}
   )
  ]
  rendererClasses [&NiagaraLightRendererProperties, &NiagaraSpriteRendererProperties]
  renderers [
   {rendererClass=&NiagaraLightRendererProperties, rendererIndex=0}
   {rendererClass=&NiagaraSpriteRendererProperties, rendererIndex=1}
  ]
 }}
summary {
 systemName=NS_TEMP_ToyFox_PedestalGlow
 emitters [{emitterName=Lights, rendererClasses [&NiagaraLightRendererProperties, &NiagaraSpriteRendererProperties]}]
 userVariables [
  User.Intensity {type=NiagaraFloat, defaultValue=0.5, description=""}
  User.LightRadiusOverride {type=NiagaraFloat, defaultValue=1, description=""}
 ]
}
