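# Temporary Niagara effect for the ToyFox cannon projectile, layering a stretched mesh, two capsule sprites, a shell burst mesh and a light.
# Each active emitter burst-spawns, then eases from User.MuzzlePosition to the owner's position using an age curve.
# Meshes and capsules orient along that line, stretch with its length and rescale when the camera looks nearly down it.
# A wide smoke trail emitter is present but disabled.
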NiagaraSystem /ToyFoxGameplay/Assets/FX/TEMP/TEMP_NS_ToyFox_Cannon_Projectile
dependencies {
 usedDataInterfaces [
  &NiagaraDataInterfaceCamera
  &NiagaraDataInterfaceCurve
  &NiagaraDataInterfaceMeshRendererInfo
  &NiagaraDataInterfaceVectorCurve
  &NiagaraDataInterfaceColorCurve
 ]
 usedDynamicInputs []
 usedModules [
  &SystemState
  &FN_Timer
  &self:SystemSpawnScript.SystemScriptSource.SystemScriptGraph.NiagaraNodeAssignment_0.SetVariables_#1
  &self:SystemSpawnScript.SystemScriptSource.SystemScriptGraph.NiagaraNodeAssignment_1.SetVariables_#2
 ]
}
emitters {
 CapsuleSprite {
  emitterUpdateScript [
   &EmitterState(
    Inactive Response=::NewEnumerator1
    Life Cycle Mode=::NewEnumerator1
    Loop Behavior=::NewEnumerator1
    Loop Duration Mode=::NewEnumerator1
    UseLoopDelay=false
   )
   &SpawnBurst_Instantaneous(Use Spawn Probability=false, UseLoopCountLimit=false)
  ]
  particleSpawnScript [&InitializeParticle(
    Color=<rgba 0 0.7031 1 0.5>
    Position Offset Coordinate Space=::Local
    Sprite Size=<200 75>
    Sprite Size Mode=::NewEnumerator4
    UsePositionOffset=true
    Write Material Random=true
   )]
  particleUpdateScript [
   &SolveForcesAndVelocity(Clamp Velocity=false, Limit Acceleration=false)
   &ParticleState(Kill Particles When Lifetime Has Elapsed=false, Let Infinitely Lived Particles Die When Emitter Deactivates=true)
   &ScaleSpriteSize(
    Scale Factor=&MakeVector2D(
     X=&MakeCustomFloatFromBool(
      Bool=&SetBoolByFloatComparison(
       A=&Abs_Float(Float:NiagaraFloat=&DotProduct(
         Vector A=&ReturnCameraVector(Camera Vectors=::NewEnumerator1, Coordinate Space=::Simulation)
         Vector B=&NormalizeVector(
          Vector To Normalize=&Subtract_Vector(A=$User.MuzzlePosition, B=Linked() {linkedVariable {name=Engine.Owner.Position, type=NiagaraPosition}})
          Safe Normalize=true
          Safe Fallback Threshold=0
          Safe Fallback Vector=<0 0 0.001>
         )
        ))
       B=0.9
      )
      True Float=1.5
      False Float=1
     )
     Y=1
    )
    Scale Sprite Size Mode=::NewEnumerator2
   ) {module=ScaleSpriteSize001}
   &UpdateMeshOrientation(
    Facing Direction=&NormalizeVector(
     Vector To Normalize=&Subtract_Vector(A=Linked() {linkedVariable {name=Engine.Owner.Position, type=NiagaraPosition}}, B=$User.MuzzlePosition)
     Safe Normalize=true
     Safe Fallback Threshold=0
     Safe Fallback Vector=<0 0 0.001>
    )
    Facing Interpolation=false
    Orientation Method=::NewEnumerator1
    Override Quat=false
    TransformFacing=false
    TransformUp=false
   )
   &ScaleSpriteSize(
    Scale Factor=&MakeVector2D(
     X=1
     Y=&MakeFloatFromVector(
      Vector=&MakeVector(
       X=&ClampFloat(
        Float=&Divide_Float(
         A=&VectorLength(Vector=&Subtract_Vector(A=$User.MuzzlePosition, B=Linked() {linkedVariable {name=Engine.Owner.Position, type=NiagaraPosition}}))
         B=350
        )
        Min=0.1
        Max=15
       )
       Y=1
       Z=1
      )
      Channel=::NewEnumerator0
     )
    )
    Scale Sprite Size Mode=::NewEnumerator2
   )
   /SetVariables(Particles.Position=&Add_Vector(
     A=&Lerp_Vector(
      A=$User.MuzzlePosition
      B=Linked() {linkedVariable {name=Engine.Owner.Position, type=NiagaraPosition}}
      Alpha=&FloatFromCurve(
       FloatCurve=DataInterface() {propertyValues {Curve {keys [
           {arriveTangent=2.265, interpMode=RCIM_Cubic, leaveTangent=2.265, tangentMode=RCTM_User, time=0, value=0}
           {arriveTangent=0.2891, interpMode=RCIM_Cubic, leaveTangent=0.2891, tangentMode=RCTM_User, time=0.25, value=1}
          ]}}}
       CurveIndex=$Particles.Age
      )
     )
     B=&TransformVector(
      Vector=&Lerp_Vector(
       A=<0 0 0>
       B=<-800 0 0>
       Alpha=&FloatFromCurve(
        FloatCurve=DataInterface() {propertyValues {Curve {keys [
            {arriveTangent=2.265, interpMode=RCIM_Cubic, leaveTangent=2.265, tangentMode=RCTM_User, time=0, value=0}
            {arriveTangent=0.2891, interpMode=RCIM_Cubic, leaveTangent=0.2891, tangentMode=RCTM_User, time=0.25, value=1}
           ]}}}
        CurveIndex=$Particles.Age
       )
      )
      Destination Space=::Simulation
     )
    )) {isSetParameters=true, module=SetVariables_#3}
   &AlignSpriteToMeshOrientation(
    Mesh Orientation Relative Sprite Alignment Vector=<1 0 0>
    Mesh Orientation Relative Sprite Facing Vector:Vector3f=<0 1 0>
    SetAlignment=true
    SetFacing=true
   )
   &DynamicMaterialParameters(
    Index 0 Param 1=1
    Index 0 Param 2=100
    Index 0 Param 3=1
    Module.Param0WriteEnabled=true
    Module.Param1WriteEnabled=true
    Module.Param2WriteEnabled=true
    Module.Param3WriteEnabled=true
   )
   &ScaleColor(
    Scale Alpha=&FloatFromCurve(FloatCurve=DataInterface() {propertyValues {Curve {keys [<at 0.05 0>,<at 0.2 1>]}}}, CurveIndex=$Particles.Age)
    ScaleA=true
    ScaleRGB=true
   )
  ]
  rendererClasses [&NiagaraSpriteRendererProperties]
  renderers [{rendererClass=&NiagaraSpriteRendererProperties}]
 }
 CapsuleSprite001 {
  emitterUpdateScript [
   &EmitterState(
    Inactive Response=::NewEnumerator1
    Life Cycle Mode=::NewEnumerator1
    Loop Behavior=::NewEnumerator1
    Loop Duration Mode=::NewEnumerator1
    UseLoopDelay=false
   )
   &SpawnBurst_Instantaneous(Use Spawn Probability=false, UseLoopCountLimit=false)
  ]
  particleSpawnScript [&InitializeParticle(
    Color=<rgba 5 4.5497 1.5 1>
    Position Offset Coordinate Space=::Local
    Sprite Size=<25 75>
    Sprite Size Mode=::NewEnumerator4
    UsePositionOffset=true
    Write Material Random=true
   )]
  particleUpdateScript [
   &SolveForcesAndVelocity(Clamp Velocity=false, Limit Acceleration=false)
   &ParticleState(Kill Particles When Lifetime Has Elapsed=false, Let Infinitely Lived Particles Die When Emitter Deactivates=true)
   &ScaleSpriteSize(
    Scale Factor=&MakeVector2D(
     X=&MakeCustomFloatFromBool(
      Bool=&SetBoolByFloatComparison(
       A=&Abs_Float(Float:NiagaraFloat=&DotProduct(
         Vector A=&ReturnCameraVector(Camera Vectors=::NewEnumerator1, Coordinate Space=::Simulation)
         Vector B=&NormalizeVector(
          Vector To Normalize=&Subtract_Vector(A=$User.MuzzlePosition, B=Linked() {linkedVariable {name=Engine.Owner.Position, type=NiagaraPosition}})
          Safe Normalize=true
          Safe Fallback Threshold=0
          Safe Fallback Vector=<0 0 0.001>
         )
        ))
       B=0.9
      )
      True Float=$System.ATS_Scale
      False Float=1
     )
     Y=1
    )
    Scale Sprite Size Mode=::NewEnumerator2
   ) {module=ScaleSpriteSize001}
   &UpdateMeshOrientation(
    Facing Direction=&NormalizeVector(
     Vector To Normalize=&Subtract_Vector(A=Linked() {linkedVariable {name=Engine.Owner.Position, type=NiagaraPosition}}, B=$User.MuzzlePosition)
     Safe Normalize=true
     Safe Fallback Threshold=0
     Safe Fallback Vector=<0 0 0.001>
    )
    Facing Interpolation=false
    Orientation Method=::NewEnumerator1
    Override Quat=false
    TransformFacing=false
    TransformUp=false
   )
   &ScaleSpriteSize(
    Scale Factor=&MakeVector2D(
     X=1
     Y=&MakeFloatFromVector(
      Vector=&MakeVector(
       X=&ClampFloat(
        Float=&Divide_Float(
         A=&VectorLength(Vector=&Subtract_Vector(A=$User.MuzzlePosition, B=Linked() {linkedVariable {name=Engine.Owner.Position, type=NiagaraPosition}}))
         B=350
        )
        Min=0.1
        Max=15
       )
       Y=1
       Z=1
      )
      Channel=::NewEnumerator0
     )
    )
    Scale Sprite Size Mode=::NewEnumerator2
   )
   /SetVariables(Particles.Position=&Add_Vector(
     A=&Lerp_Vector(
      A=$User.MuzzlePosition
      B=Linked() {linkedVariable {name=Engine.Owner.Position, type=NiagaraPosition}}
      Alpha=&FloatFromCurve(
       FloatCurve=DataInterface() {propertyValues {Curve {keys [
           {arriveTangent=2.265, interpMode=RCIM_Cubic, leaveTangent=2.265, tangentMode=RCTM_User, time=0, value=0}
           {arriveTangent=0.2891, interpMode=RCIM_Cubic, leaveTangent=0.2891, tangentMode=RCTM_User, time=0.25, value=1}
          ]}}}
       CurveIndex=$Particles.Age
      )
     )
     B=&TransformVector(
      Vector=&Lerp_Vector(
       A=<0 0 0>
       B=<-800 0 0>
       Alpha=&FloatFromCurve(
        FloatCurve=DataInterface() {propertyValues {Curve {keys [
            {arriveTangent=2.265, interpMode=RCIM_Cubic, leaveTangent=2.265, tangentMode=RCTM_User, time=0, value=0}
            {arriveTangent=0.2891, interpMode=RCIM_Cubic, leaveTangent=0.2891, tangentMode=RCTM_User, time=0.25, value=1}
           ]}}}
        CurveIndex=$Particles.Age
       )
      )
      Destination Space=::Simulation
     )
    )) {isSetParameters=true, module=SetVariables_#3}
   &AlignSpriteToMeshOrientation(
    Mesh Orientation Relative Sprite Alignment Vector=<1 0 0>
    Mesh Orientation Relative Sprite Facing Vector:Vector3f=<0 1 0>
    SetAlignment=true
    SetFacing=true
   )
   &DynamicMaterialParameters(
    Index 0 Param 1=1
    Index 0 Param 2=100
    Index 0 Param 3=1
    Module.Param0WriteEnabled=true
    Module.Param1WriteEnabled=true
    Module.Param2WriteEnabled=true
    Module.Param3WriteEnabled=true
   )
   &ScaleColor(
    Scale Alpha=&FloatFromCurve(FloatCurve=DataInterface() {propertyValues {Curve {keys [<at 0.05 0>,<at 0.2 1>]}}}, CurveIndex=$Particles.Age)
    ScaleA=true
    ScaleRGB=true
   )
  ]
  rendererClasses [&NiagaraSpriteRendererProperties]
  renderers [{rendererClass=&NiagaraSpriteRendererProperties}]
 }
 Light {
  emitterUpdateScript [
   &EmitterState(
    Inactive Response=::NewEnumerator1
    Life Cycle Mode=::NewEnumerator1
    Loop Behavior=::NewEnumerator1
    Loop Duration Mode=::NewEnumerator1
    UseLoopDelay=false
   )
   &SpawnBurst_Instantaneous(Use Spawn Probability=false, UseLoopCountLimit=false)
  ]
  particleSpawnScript [&InitializeParticle(Color=<rgba 0 1.971 2 1>, Lifetime=5, UsePositionOffset=false, Write Material Random=true)]
  particleUpdateScript [
   &ParticleState(Kill Particles When Lifetime Has Elapsed=false, Let Infinitely Lived Particles Die When Emitter Deactivates=true)
   /SetVariables(Particles.Position=&Lerp_Vector(
     A=$User.MuzzlePosition
     B=Linked() {linkedVariable {name=Engine.Owner.Position, type=NiagaraPosition}}
     Alpha=&FloatFromCurve(
      FloatCurve=DataInterface() {propertyValues {Curve {keys [
          {arriveTangent=2.265, interpMode=RCIM_Cubic, leaveTangent=2.265, tangentMode=RCTM_User, time=0, value=0}
          {arriveTangent=0.2891, interpMode=RCIM_Cubic, leaveTangent=0.2891, tangentMode=RCTM_User, time=0.25, value=1}
         ]}}}
      CurveIndex=$Particles.Age
     )
    )) {isSetParameters=true}
   &FN_Light_Attributes(Create Exponent Attribute=true, Create Radius Attribute=true, Exponent=2, Radius=500)
   &ScaleColor(
    Scale RGB=&VectorFromFloat(Value=&FloatFromCurve(
      FloatCurve=DataInterface() {propertyValues {Curve {keys [{interpMode=RCIM_Cubic, tangentMode=RCTM_User, time=0.3, value=1}]}}}
      CurveIndex=$Particles.Age
     ))
    ScaleA=true
    ScaleRGB=true
   )
  ]
  rendererClasses [&NiagaraSpriteRendererProperties, &NiagaraLightRendererProperties]
  renderers [
   {rendererClass=&NiagaraSpriteRendererProperties}
   {rendererClass=&NiagaraLightRendererProperties, rendererIndex=1}
  ]
 }
 Projectile_Mesh {
  emitterUpdateScript [
   &EmitterState(
    Inactive Response=::NewEnumerator1
    Life Cycle Mode=::NewEnumerator1
    Loop Behavior=::NewEnumerator1
    Loop Duration Mode=::NewEnumerator1
    UseLoopDelay=false
   )
   &SpawnBurst_Instantaneous(Use Spawn Probability=false, UseLoopCountLimit=false)
  ]
  particleSpawnScript [
   &InitializeParticle(
    Color=<rgba 0 35.1565 50 1>
    Mesh Scale=<4 6 6>
    Mesh Scale Mode=::NewEnumerator4
    Sprite Size Mode=::NewEnumerator3
    Uniform Sprite Size=50
    UsePositionOffset=false
    Write Material Random=true
   )
   &SystemLocation(MaskSpawnGroup=false, Offset Coordinate Space=::World)
   &InitializeParticle(
    Color Mode=::NewEnumerator0
    Lifetime=&Lerp_Float(
     A=0.1
     B=1
     Alpha=&RemapRange(Input Value To Remap=$User.ScanLength, Input Min=0, Input Max=1000, Output Min=0, Output Max=1)
    )
    Mesh Scale Mode=::NewEnumerator4
    Sprite Size Mode=::NewEnumerator3
    UsePositionOffset=false
    Write Material Random=true
   ) {enabled=false, module=InitializeParticle001}
  ]
  particleUpdateScript [
   &ParticleState(Kill Particles When Lifetime Has Elapsed=false, Let Infinitely Lived Particles Die When Emitter Deactivates=true)
   &ScaleMeshSize(Scale Factor=&MakeVector(
     X=1
     Y=&MakeCustomFloatFromBool(
      Bool=&SetBoolByFloatComparison(
       A=&Abs_Float(Float:NiagaraFloat=&DotProduct(
         Vector A=&ReturnCameraVector(Camera Vectors=::NewEnumerator1, Coordinate Space=::Simulation)
         Vector B=&NormalizeVector(
          Vector To Normalize=&Subtract_Vector(A=$User.MuzzlePosition, B=Linked() {linkedVariable {name=Engine.Owner.Position, type=NiagaraPosition}})
          Safe Normalize=true
          Safe Fallback Threshold=0
          Safe Fallback Vector=<0 0 0.001>
         )
        ))
       B=0.9
      )
      True Float=$System.ATS_Scale
      False Float=1
     )
     Z=&MakeCustomFloatFromBool(
      Bool=&SetBoolByFloatComparison(
       A=&Abs_Float(Float:NiagaraFloat=&DotProduct(
         Vector A=&ReturnCameraVector(Camera Vectors=::NewEnumerator1, Coordinate Space=::Simulation)
         Vector B=&NormalizeVector(
          Vector To Normalize=&Subtract_Vector(A=$User.MuzzlePosition, B=Linked() {linkedVariable {name=Engine.Owner.Position, type=NiagaraPosition}})
          Safe Normalize=true
          Safe Fallback Threshold=0
          Safe Fallback Vector=<0 0 0.001>
         )
        ))
       B=0.9
      )
      True Float=$System.ATS_Scale
      False Float=1
     )
    )) {module=ScaleMeshSize001}
   &DynamicMaterialParameters(
    Index 0 Param 1=$System.Muzzle_Pos_X
    Index 0 Param 2=$System.Muzzle_Pos_Y
    Index 0 Param 3=$System.Muzzle_Pos_Z
    Index 0 Param 4=&FloatFromCurve(FloatCurve=DataInterface() {propertyValues {Curve {keys [<at 0 0>,<at 0.1 1>,<at 1 5>]}}}, CurveIndex=$Particles.Age, Scale Curve=1000)
    Module.Param0WriteEnabled=true
    Module.Param1WriteEnabled=true
    Module.Param2WriteEnabled=true
    Module.Param3WriteEnabled=true
   )
   &ScaleMeshSize(Scale Factor=&MakeVector(
     X=&ClampFloat(
      Float=&Divide_Float(
       A=&VectorLength(Vector=&Subtract_Vector(A=$User.MuzzlePosition, B=Linked() {linkedVariable {name=Engine.Owner.Position, type=NiagaraPosition}}))
       B=350
      )
      Min=0.1
      Max=10
     )
     Y=1
     Z=1
    ))
   &UpdateMeshOrientation(
    Facing Direction=&NormalizeVector(
     Vector To Normalize=&Subtract_Vector(A=Linked() {linkedVariable {name=Engine.Owner.Position, type=NiagaraPosition}}, B=$User.MuzzlePosition)
     Safe Normalize=true
     Safe Fallback Threshold=0
     Safe Fallback Vector=<0 0 0.001>
    )
    Facing Interpolation=false
    Orientation Method=::NewEnumerator1
    Override Quat=false
    TransformFacing=false
    TransformUp=false
   )
   /SetVariables(Particles.Position=&Lerp_Vector(
     A=$User.MuzzlePosition
     B=Linked() {linkedVariable {name=Engine.Owner.Position, type=NiagaraPosition}}
     Alpha=&FloatFromCurve(
      FloatCurve=DataInterface() {propertyValues {Curve {keys [
          {arriveTangent=2.265, interpMode=RCIM_Cubic, leaveTangent=2.265, tangentMode=RCTM_User, time=0, value=0}
          {arriveTangent=0.2891, interpMode=RCIM_Cubic, leaveTangent=0.2891, tangentMode=RCTM_User, time=0.25, value=1}
         ]}}}
      CurveIndex=$Particles.Age
     )
    )) {isSetParameters=true, module=SetVariables_#3}
   &ScaleColor(
    Scale Alpha=&FloatFromCurve(FloatCurve=DataInterface() {propertyValues {Curve {keys [<at 0.2 0>,<at 0.3 1>]}}})
    ScaleA=true
    ScaleRGB=true
   )
  ]
  rendererClasses [&NiagaraMeshRendererProperties]
  renderers [{rendererClass=&NiagaraMeshRendererProperties, rendererIndex=0}]
 }
 ShellBurst001 {
  emitterUpdateScript [
   &EmitterState(
    Inactive Response=::NewEnumerator1
    Life Cycle Mode=::NewEnumerator1
    Loop Behavior=::NewEnumerator1
    Loop Duration Mode=::NewEnumerator1
    UseLoopDelay=true
   )
   &SpawnBurst_Instantaneous(Spawn Time=0.02, Use Spawn Probability=false, UseLoopCountLimit=false)
  ]
  particleSpawnScript [
   &InitializeParticle(
    Color=<rgba 6 6 6 1>
    Lifetime=0.15
    Mesh Scale=<1 0.3 0.3>
    Mesh Scale Mode=::NewEnumerator4
    Position Offset Coordinate Space=::Local
    UsePositionOffset=true
    Write Material Random=true
   )
   &InitialMeshOrientation(Orientation Vector=<-1 0 0>, Use Orientation Vector=true, Use Rotation Vector=false)
   &DynamicMaterialParameters(
    Index 0 Param 1=0.5
    Index 0 Param 2=1
    Index 0 Param 3=1
    Index 0 Param 4=1
    Module.Param0WriteEnabled=true
    Module.Param1WriteEnabled=true
    Module.Param2WriteEnabled=true
    Module.Param3WriteEnabled=true
   )
  ]
  particleUpdateScript [
   &ParticleState(Kill Particles When Lifetime Has Elapsed=false, Let Infinitely Lived Particles Die When Emitter Deactivates=true)
   &ScaleMeshSize(Scale Factor=&MakeVector(
     X=1
     Y=&MakeCustomFloatFromBool(
      Bool=&SetBoolByFloatComparison(
       A=&Abs_Float(Float:NiagaraFloat=&DotProduct(
         Vector A=&ReturnCameraVector(Camera Vectors=::NewEnumerator1, Coordinate Space=::Simulation)
         Vector B=&NormalizeVector(
          Vector To Normalize=&Subtract_Vector(A=$User.MuzzlePosition, B=Linked() {linkedVariable {name=Engine.Owner.Position, type=NiagaraPosition}})
          Safe Normalize=true
          Safe Fallback Threshold=0
          Safe Fallback Vector=<0 0 0.001>
         )
        ))
       B=0.9
      )
      True Float=$System.ATS_Scale
      False Float=0
     )
     Z=&MakeCustomFloatFromBool(
      Bool=&SetBoolByFloatComparison(
       A=&Abs_Float(Float:NiagaraFloat=&DotProduct(
         Vector A=&ReturnCameraVector(Camera Vectors=::NewEnumerator1, Coordinate Space=::Simulation)
         Vector B=&NormalizeVector(
          Vector To Normalize=&Subtract_Vector(A=$User.MuzzlePosition, B=Linked() {linkedVariable {name=Engine.Owner.Position, type=NiagaraPosition}})
          Safe Normalize=true
          Safe Fallback Threshold=0
          Safe Fallback Vector=<0 0 0.001>
         )
        ))
       B=0.9
      )
      True Float=$System.ATS_Scale
      False Float=0
     )
    )) {module=ScaleMeshSize001}
   &UpdateMeshOrientation(
    Facing Direction=&NormalizeVector(
     Vector To Normalize=&Subtract_Vector(A=$User.MuzzlePosition, B=Linked() {linkedVariable {name=Engine.Owner.Position, type=NiagaraPosition}})
     Safe Normalize=true
     Safe Fallback Threshold=0
     Safe Fallback Vector=<0 0 0.001>
    )
    Facing Interpolation=false
    Orientation Method=::NewEnumerator1
    Override Quat=false
    TransformFacing=false
    TransformUp=false
   )
   &ScaleMeshSize(Scale Factor=&VectorFromCurve(VectorCurve=DataInterface() {propertyValues {
       XCurve {keys [
         {arriveTangent=1.6643, interpMode=RCIM_Cubic, leaveTangent=1.6643, tangentMode=RCTM_User, time=0, value=0}
         {arriveTangent=0.3199, interpMode=RCIM_Cubic, leaveTangent=0.3199, tangentMode=RCTM_User, time=1, value=1}
        ]}
       YCurve {keys [{arriveTangent=3.3122, interpMode=RCIM_Cubic, leaveTangent=3.3122, tangentMode=RCTM_User, time=0, value=0}, <at 1 1 RCIM_Cubic>]}
       ZCurve {keys [{arriveTangent=3.3122, interpMode=RCIM_Cubic, leaveTangent=3.3122, tangentMode=RCTM_User, time=0, value=0}, <at 1 1 RCIM_Cubic>]}
      }}))
   &ScaleMeshSize(Scale Factor=&MakeVector(
     X=&ClampFloat(
      Float=&Divide_Float(
       A=&VectorLength(Vector=&Subtract_Vector(A=$User.MuzzlePosition, B=Linked() {linkedVariable {name=Engine.Owner.Position, type=NiagaraPosition}}))
       B=100
      )
      Min=0.1
      Max=10
     )
     Y=1
     Z=1
    )) {module=ScaleMeshSize002}
   &ScaleColor(
    Scale Alpha=&FloatFromCurve(FloatCurve=DataInterface() {propertyValues {Curve {keys [<at 0 0>,<at 0.2 1>]}}})
    Scale RGB=&Multiply_Vector(
     A=<1 1 1>
     B=&MakeVectorFromLinearColorRGB(Color=&ColorFromCurve(ColorCurve=DataInterface() {propertyValues {AlphaCurve {keys [<at 0 1 RCIM_Cubic>]}, BlueCurve {keys [<at 0 1>]}, GreenCurve {keys [<at 0 0.9855>]}, RedCurve {keys [<at 0 0>]}}}))
    )
    ScaleA=true
    ScaleRGB=true
   )
   /SetVariables(Particles.Position=&Lerp_Vector(
     A=$User.MuzzlePosition
     B=Linked() {linkedVariable {name=Engine.Owner.Position, type=NiagaraPosition}}
     Alpha=&FloatFromCurve(
      FloatCurve=DataInterface() {propertyValues {Curve {keys [
          {arriveTangent=2.265, interpMode=RCIM_Cubic, leaveTangent=2.265, tangentMode=RCTM_User, time=0, value=0}
          {arriveTangent=0.2891, interpMode=RCIM_Cubic, leaveTangent=0.2891, tangentMode=RCTM_User, time=0.25, value=1}
         ]}}}
      CurveIndex=$Particles.Age
     )
    )) {isSetParameters=true, module=SetVariables_#3}
   &ScaleColor(
    Scale Alpha=&FNTimeOfDayScalar(Night Scale=0.5, Morning Scale=1, Evening Scale=1)
    ScaleA=true
    ScaleRGB=false
   ) {module=ScaleColor003}
   &ScaleColor(
    Scale Alpha=&Multiply_Float(
     A=&FloatFromCurve(FloatCurve=DataInterface() {propertyValues {Curve {keys [<at 0.15 0>,<at 0.25 1>]}}}, CurveIndex=$Particles.Age)
     B=&FN_LOD_Distance_Remap(Start Distance=0, End Distance=20000, Minimum Output=1, Maximum Output=0, Clamp Output=true)
    )
    ScaleA=true
    ScaleRGB=true
   ) {module=ScaleColor001}
  ]
  rendererClasses [&NiagaraMeshRendererProperties]
  renderers [{rendererClass=&NiagaraMeshRendererProperties, rendererIndex=0}]
 }
 Smoke_Wide_Large {
  bEnabled=false
  emitterUpdateScript [
   &EmitterState(Delay First Loop Only=true, Life Cycle Mode=::NewEnumerator1, UseLoopDelay=true)
   &SpawnRate(SpawnRate=250, Use Spawn Probability=false)
  ]
  particleSpawnScript [
   /Niagara/Modules/Spawn/Initialization/InitializeParticle(
    Color=&UniformRangedLinearColor(Minimum=<rgba 0.901 0.9487 1 0.1>, Maximum=<rgba 0.6175 0.7014 0.7917 0.15>)
    Lifetime=&UniformRangedFloat(Minimum=0.4, Maximum=0.6)
    Module.Write Color=true
    Module.Write Lifetime=true
    Module.Write Mass=true
    Module.Write Material Random=true
    Module.Write Position=true
    Module.Write Scale=false
    Module.Write SpriteRotation=true
    Module.Write SpriteSize=true
    Position=/Niagara/DynamicInputs/Helpers/SimulationPosition()
    Sprite Rotation=&RandomRangeFloat(Minimum=0, Maximum=1)
    Sprite Size=&Vector2DFromFloat(Value=&RandomRangeFloat(Minimum=125, Maximum=175))
   )
   /SetVariables(Particles.Position=&Lerp_Vector(
     A=$User.MuzzlePosition
     B=Linked() {linkedVariable {name=Engine.Owner.Position, type=NiagaraPosition}}
     Alpha=&FloatFromCurve(
      FloatCurve=DataInterface() {propertyValues {Curve {keys [
          {arriveTangent=2.265, interpMode=RCIM_Cubic, leaveTangent=2.265, tangentMode=RCTM_User, time=0, value=0}
          {arriveTangent=0.2891, interpMode=RCIM_Cubic, leaveTangent=0.2891, tangentMode=RCTM_User, time=0.25, value=1}
         ]}}}
      CurveIndex=$Emitter.Age
     )
    )) {isSetParameters=true}
   /SetVariables(Particles.InitSpriteRotRate=&UniformRangedFloat(Minimum=-3, Maximum=3)) {isSetParameters=true}
   &AddVelocityInCone(
    Cone Angle=4
    Module.Use Velocity Falloff On Cone Axis=false
    Velocity Strength=&RandomRangeFloat(Minimum=-250, Maximum=-250)
   )
   &SphereLocation(
    MaskSpawnGroup=false
    Override Seed=false
    Sphere Radius=20
    Surface Only=false
    UseNonUniformScale=false
    UseOffset=false
    UseSphereOrientation=false
   )
  ]
  particleUpdateScript [
   &ParticleState()
   &SpriteRotationRate(Rotation Rate=&Multiply_Float(
     A=$Particles.InitSpriteRotRate
     B=&FloatFromCurve(
      FloatCurve=DataInterface() {propertyValues {Curve {keys [
          {arriveTangent=-1.9214, interpMode=RCIM_Cubic, leaveTangent=-1.9214, tangentMode=RCTM_User, time=-0.0046, value=1.0113}
          {arriveTangent=-0.2496, interpMode=RCIM_Cubic, leaveTangent=-0.2496, tangentMode=RCTM_User, time=0.9896, value=0.16}
         ]}}}
      CurveIndex=$Emitter.Age
      Scale Curve=150
     )
    ))
   &ScaleColor(
    Scale Alpha=&FloatFromCurve(FloatCurve=DataInterface() {propertyValues {Curve {keys [
         <at 0 0>
         <at 0.2 0.8>
         {arriveTangent=-0.0528, interpMode=RCIM_Cubic, leaveTangent=-0.0528, tangentMode=RCTM_User, time=0.85, value=0.5}
         {interpMode=RCIM_Cubic, tangentMode=RCTM_Break, time=1, value=0}
        ]}}})
    Scale RGB=&VectorFromFloat(Value=&FloatFromCurve(FloatCurve=DataInterface() {propertyValues {Curve {keys [<at 0.1505 1>,<at 1 0>,<at 5 2>]}}}))
    ScaleA=true
    ScaleRGB=true
   )
   &Drag(
    Drag=&FloatFromCurve(FloatCurve=DataInterface() {propertyValues {Curve {keys [<at 0 1>,<at 1 0.5>]}}}, Scale Curve=4)
    Use Linear Drag=true
    Use Rotational Drag=false
   )
   &GravityForce(Gravity=&Lerp_Vector(A=<0 0 -200>, B=<0 0 -200>, Alpha=$Particles.MaterialRandom))
   /Niagara/Modules/Update/Forces/CurlNoiseForce(Noise Frequency=0.01, Noise Strength=200, Pan Noise=true, Pan Noise Field=<0 0 1>)
   &AccelerationForce(Acceleration=<0 200 0>, Coordinate Space=::Local)
   &SolveForcesAndVelocity(Clamp Velocity=false, Limit Acceleration=false)
   &ScaleSpriteSize(Scale Factor=&Vector2DFromFloat(Value=&FloatFromCurve(FloatCurve=DataInterface() {propertyValues {Curve {keys [
          {arriveTangent=7.4766, interpMode=RCIM_Cubic, leaveTangent=5.0379, tangentMode=RCTM_User, time=0, value=0.077}
          <at 0.1474 0.4691>
          <at 1 1 RCIM_Cubic>
         ]}}})))
   &DynamicMaterialParameters(
    Index 0 Param 1=&Add_Float(
     A=$Particles.RandomFloat
     B=&FloatFromCurve(
      FloatCurve=DataInterface() {propertyValues {Curve {keys [
          {arriveTangent=2.7533, interpMode=RCIM_Cubic, leaveTangent=1.8958, tangentMode=RCTM_User, time=0, value=1}
          {arriveTangent=0.8292, interpMode=RCIM_Cubic, leaveTangent=0.8292, tangentMode=RCTM_User, time=0.321, value=1.3453}
          {arriveTangent=0.0529, interpMode=RCIM_Cubic, leaveTangent=0.2021, tangentMode=RCTM_User, time=1, value=1.5}
         ]}}}
      CurveIndex=&Add_Float(A=&Multiply_Float(A=$Particles.Initial.MaterialRandom, B=0.5), B=$Particles.NormalizedAge)
     )
    )
    Index 0 Param 2=&FloatFromCurve(FloatCurve=DataInterface() {propertyValues {Curve {keys [
         {arriveTangent=0.0453, interpMode=RCIM_Cubic, tangentMode=RCTM_Break, time=0, value=-1}
         {arriveTangent=0.0604, interpMode=RCIM_Cubic, leaveTangent=0.0604, tangentMode=RCTM_User, time=1, value=-1}
        ]}}})
    Index 0 Param 3=&FloatFromCurve(FloatCurve=DataInterface() {propertyValues {Curve {keys [{arriveTangent=-53.9199, interpMode=RCIM_Cubic, time=0, value=0.5}, {interpMode=RCIM_Cubic, leaveTangent=0.2723, time=0.1, value=0}]}}})
    Index 0 Param 4=&FloatFromCurve(FloatCurve=DataInterface() {propertyValues {Curve {keys [{interpMode=RCIM_Cubic, leaveTangent=-1.5927, tangentMode=RCTM_Break, time=0, value=1}, <at 1 0>]}}})
    Index 1 Param 1=&FloatFromCurve(
     FloatCurve=DataInterface() {propertyValues {Curve {keys [
         {interpMode=RCIM_Cubic, leaveTangent=4.7088, tangentMode=RCTM_Break, time=0, value=0}
         {arriveTangent=4.2257, interpMode=RCIM_Cubic, leaveTangent=4.2257, tangentMode=RCTM_User, time=0.25, value=1}
        ]}}}
     Scale Curve=&RandomRangeFloat(Minimum=1.5, Maximum=1.5)
    )
    Index 1 Param 2=&FloatFromCurve(
     FloatCurve=DataInterface() {propertyValues {Curve {keys [<at 0 0.2>, {leaveTangent=0.9066, time=1, value=2}]}}}
     Scale Curve=&RandomRangeFloat(Minimum=0.5, Maximum=0.6)
    )
    Index 1 Param 3=&FloatFromCurve(FloatCurve=DataInterface() {propertyValues {Curve {keys [
         {interpMode=RCIM_Cubic, leaveTangent=6.3917, tangentMode=RCTM_Break, time=0, value=0.2}
         {arriveTangent=4.8433, interpMode=RCIM_Cubic, leaveTangent=19.0929, tangentMode=RCTM_User, time=0.25, value=1}
        ]}}})
    Index 1 Param 4=&FloatFromCurve(
     FloatCurve=DataInterface() {propertyValues {Curve {keys [{arriveTangent=-0.3224, interpMode=RCIM_Cubic, leaveTangent=-0.3224, tangentMode=RCTM_User, time=0, value=1}, <at 1 0.9>]}}}
     Scale Curve=&RandomRangeFloat(Minimum=1.1, Maximum=1.4)
    )
    Index 2 Param 1=&FloatFromCurve(FloatCurve=DataInterface() {propertyValues {Curve {keys [
         {interpMode=RCIM_Cubic, leaveTangent=4.2485, tangentMode=RCTM_Break, time=0, value=0}
         {arriveTangent=1.0548, interpMode=RCIM_Cubic, leaveTangent=1.0548, tangentMode=RCTM_User, time=0.17, value=0.5715}
         <at 1 1>
        ]}}})
    Index 2 Param 2=&FloatFromCurve(
     FloatCurve=DataInterface() {propertyValues {Curve {keys [
         {arriveTangent=3.069, interpMode=RCIM_Cubic, leaveTangent=3.069, tangentMode=RCTM_User, time=0, value=0.2}
         {arriveTangent=0.0259, interpMode=RCIM_Cubic, leaveTangent=0.0259, tangentMode=RCTM_User, time=1, value=1.4}
        ]}}}
     Scale Curve=&RandomRangeFloat(Minimum=0.2, Maximum=1.25)
    )
    Index 2 Param 3=&FloatFromCurve(FloatCurve=DataInterface() {propertyValues {Curve {keys [
         {interpMode=RCIM_Cubic, leaveTangent=2.3064, tangentMode=RCTM_Break, time=0, value=1}
         {interpMode=RCIM_Cubic, leaveTangent=11.1233, tangentMode=RCTM_Break, time=1, value=2}
        ]}}})
    Index 2 Param 4=&FloatFromCurve(FloatCurve=DataInterface() {propertyValues {Curve {keys [
         {arriveTangent=-0.6867, interpMode=RCIM_Cubic, leaveTangent=-0.6867, tangentMode=RCTM_User, time=0, value=1}
         {interpMode=RCIM_Cubic, leaveTangent=-1.2956, tangentMode=RCTM_Break, time=1, value=0.5}
        ]}}})
    Module.Param0WriteEnabled=true
    Module.Param0WriteEnabled1=true
    Module.Param0WriteEnabled2=true
    Module.Param1WriteEnabled=false
    Module.Param1WriteEnabled1=true
    Module.Param1WriteEnabled2=true
    Module.Param2WriteEnabled=true
    Module.Param2WriteEnabled1=true
    Module.Param2WriteEnabled2=true
    Module.Param3WriteEnabled=true
    Module.Param3WriteEnabled1=true
    Module.Param3WriteEnabled2=true
    Write Parameter Index 1=true
    Write Parameter Index 2=true
   )
   &ScaleColor(
    Scale RGB=&MakeVectorFromLinearColorRGB(Color=&ColorFromCurve(ColorCurve=DataInterface() {propertyValues {
        AlphaCurve {keys [<at 0 1 RCIM_Cubic>,<at 1 0 RCIM_Cubic>]}
        BlueCurve {keys [<at 0 1 RCIM_Cubic>,<at 0.0721 1 RCIM_Cubic>]}
        GreenCurve {keys [<at 0 0.5282 RCIM_Cubic>,<at 0.0721 1 RCIM_Cubic>]}
        RedCurve {keys [<at 0 0.3 RCIM_Cubic>,<at 0.0721 1 RCIM_Cubic>]}
       }}))
    ScaleA=true
    ScaleRGB=true
   ) {module=ScaleColor001}
  ]
  rendererClasses [&NiagaraSpriteRendererProperties]
  renderers [
   {rendererClass=&NiagaraSpriteRendererProperties}
   {rendererClass=&NiagaraSpriteRendererProperties, rendererIndex=1}
  ]
 }
}
summary {
 systemName=TEMP_NS_ToyFox_Cannon_Projectile
 emitters [
  {emitterName=Projectile_Mesh, rendererClasses [&NiagaraMeshRendererProperties]}
  {emitterName=CapsuleSprite, rendererClasses [&NiagaraSpriteRendererProperties]}
  {emitterName=CapsuleSprite001, rendererClasses [&NiagaraSpriteRendererProperties]}
  {emitterName=ShellBurst001, rendererClasses [&NiagaraMeshRendererProperties]}
  {emitterName=Light, rendererClasses [&NiagaraSpriteRendererProperties, &NiagaraLightRendererProperties]}
  {bEnabled=false, emitterName=Smoke_Wide_Large, rendererClasses [&NiagaraSpriteRendererProperties]}
 ]
 userVariables [
  User.ActorPosition {type=Vector3f, description="", defaultValue {struct=Vector3f, value=<0 0 0>}}
  User.ForwardVector {type=Vector3f, description="", defaultValue {struct=Vector3f, value=<0 0 0>}}
  User.MuzzlePosition {type=Vector3f, description="", defaultValue {struct=Vector3f, value=<0 0 0>}}
  User.OriginOffset {type=Vector3f, description="", defaultValue {struct=Vector3f, value=<0 0 0>}}
  User.ScanLength {type=NiagaraFloat, defaultValue=0, description=""}
  User.StartTime {type=NiagaraFloat, defaultValue=0, description=""}
 ]
}
